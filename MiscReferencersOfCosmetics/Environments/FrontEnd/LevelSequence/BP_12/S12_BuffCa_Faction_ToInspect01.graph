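# Level sequence for a front-end camera transition, named as the move to the first inspect view in the S12 BuffCa faction scene.
# It animates the root of a possessed Frontend_EventLevel_Camera_C with a short keyed move and yaw swing, holding pitch and roll fixed.
# Its child CameraComponent keeps a constant field of view, and two event trigger tracks each fire one event.
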LevelSequence /MiscReferencersOfCosmetics/Environments/FrontEnd/LevelSequence/BP_12/S12_BuffCa_Faction_ToInspect01
bindings [
 CameraComponent {
  id=#1
  parent_id=#2
  possessed_class=CameraComponent
  tracks [{
    type=MovieSceneFloatTrack
    display_name=Field Of View
    sections [{type=MovieSceneFloatSection, channels [None {type=MovieSceneScriptingFloatChannel, default=80, const [80,2]}]}]
   }]
 }
 Camera_Root_Movement {
  id=#2
  possessed_class=Frontend_EventLevel_Camera_C
  tracks [{
    type=MovieScene3DTransformTrack
    sections [{
      type=MovieScene3DTransformSection
      channels [
       Location.X {type=MovieSceneScriptingDoubleChannel, default=2672.5122, keys [[0,2710],[10,2672.5122]]}
       Location.Y {type=MovieSceneScriptingDoubleChannel, default=16478.3828, keys [[0,16441.2559],[10,16478.3828]]}
       Location.Z {type=MovieSceneScriptingDoubleChannel, default=967.2159, keys [[0,975],[10,967.2159]]}
       Rotation.X {type=MovieSceneScriptingDoubleChannel, default=0.0001, const [0.0001,2]}
       Rotation.Y {type=MovieSceneScriptingDoubleChannel, default=-1.6, const [-1.6,2]}
       Rotation.Z {type=MovieSceneScriptingDoubleChannel, default=-201.6004, keys [[0,-221.7053],[10,-201.6004]]}
       Scale.X {type=MovieSceneScriptingDoubleChannel, default=1}
       Scale.Y {type=MovieSceneScriptingDoubleChannel, default=1}
       Scale.Z {type=MovieSceneScriptingDoubleChannel, default=1}
      ]
     }]
   }]
 }
]
display_rate {denominator=1, numerator=30}
tick_resolution {denominator=1, numerator=24000}
tracks [
 {
  type=MovieSceneEventTrack
  sections [{
    type=MovieSceneEventTriggerSection
    channels [None {type=MovieSceneScriptingEventChannel, const ["<Struct 'MovieSceneEvent' (0x8519a7f3480) {}>", 1]}]
   }]
 }
 {
  type=MovieSceneEventTrack
  sections [{
    type=MovieSceneEventTriggerSection
    channels [None {type=MovieSceneScriptingEventChannel, const ["<Struct 'MovieSceneEvent' (0x8519a7f3c00) {}>", 1]}]
   }]
 }
]
