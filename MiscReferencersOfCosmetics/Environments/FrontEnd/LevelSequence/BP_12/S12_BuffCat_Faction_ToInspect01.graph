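# Level sequence that moves the front-end event camera into an inspect view, named for the BuffCat faction in the BP_12 set.
# A transform track on the possessed Frontend_EventLevel_Camera_C eases or blends its location and rotation between two keys over a short move.
# The child CameraComponent holds a constant field of view.
# Two root event trigger tracks fire events, whose target functions are not shown.
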LevelSequence /MiscReferencersOfCosmetics/Environments/FrontEnd/LevelSequence/BP_12/S12_BuffCat_Faction_ToInspect01
bindings [
 CameraComponent {
  id=#1
  parent_id=#2
  possessed_class=CameraComponent
  tracks [{
    type=MovieSceneFloatTrack
    display_name=Field Of View
    sections [{type=MovieSceneFloatSection, channels [None {type=MovieSceneScriptingFloatChannel, default=80, const [80,2]}]}]
   }]
 }
 Camera_Root_Movement {
  id=#2
  possessed_class=Frontend_EventLevel_Camera_C
  tracks [{
    type=MovieScene3DTransformTrack
    sections [{
      type=MovieScene3DTransformSection
      channels [
       Location.X {type=MovieSceneScriptingDoubleChannel, default=2633.0825, keys [[0,2716.1272],[10,2633.0825]]}
       Location.Y {type=MovieSceneScriptingDoubleChannel, default=16376.0273, keys [[0,16448.2148],[10,16376.0273]]}
       Location.Z {type=MovieSceneScriptingDoubleChannel, default=964.8707, keys [[0,984.9739],[10,964.8707]]}
       Rotation.X {type=MovieSceneScriptingDoubleChannel, default=0.1753, keys [[0,0.0001],[10,0.1753]]}
       Rotation.Y {type=MovieSceneScriptingDoubleChannel, default=3.4739, keys [[0,-1.6],[10,3.4739]]}
       Rotation.Z {type=MovieSceneScriptingDoubleChannel, default=-220.5444, keys [[0,-201.6004],[10,-220.5444]]}
       Scale.X {type=MovieSceneScriptingDoubleChannel, default=1}
       Scale.Y {type=MovieSceneScriptingDoubleChannel, default=1}
       Scale.Z {type=MovieSceneScriptingDoubleChannel, default=1}
      ]
     }]
   }]
 }
]
display_rate {denominator=1, numerator=30}
tick_resolution {denominator=1, numerator=24000}
tracks [
 {
  type=MovieSceneEventTrack
  sections [{
    type=MovieSceneEventTriggerSection
    channels [None {type=MovieSceneScriptingEventChannel, const ["<Struct 'MovieSceneEvent' (0x8519a7f3840) {}>", 1]}]
   }]
 }
 {
  type=MovieSceneEventTrack
  sections [{
    type=MovieSceneEventTriggerSection
    channels [None {type=MovieSceneScriptingEventChannel, const ["<Struct 'MovieSceneEvent' (0x8519a7f3660) {}>", 1]}]
   }]
 }
]
