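# Front-end battle pass camera transition, named for moving from the Buffcat (Meowscles) room to the elevator.
# Keyframes move the camera root across the level in one fast leg, hold briefly, then ease to a final framing at fixed FOV.
# During that last leg, one elevator's door material animates its Open parameter from 0 to 1.
# Event tracks fire on the other elevator transition, the Meowscles and Maproom light switchers, and at sequence level.
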LevelSequence /MiscReferencersOfCosmetics/Environments/FrontEnd/LevelSequence/BP_12/S12_BuffcatRoom_ToElevator
bindings [
 Doors {id=#1, parent_id=#2, possessed_class=StaticMeshComponent, tracks []}
 CameraComponent {
  id=#3
  parent_id=#4
  possessed_class=CameraComponent
  tracks [{
    type=MovieSceneFloatTrack
    display_name=Field Of View
    sections [{type=MovieSceneFloatSection, channels [None {type=MovieSceneScriptingFloatChannel, default=80, const [80,2]}]}]
   }]
 }
 BP_Battlepass_LightSwitcher_Meowscles {
  id=#5
  possessed_class=BP_Battlepass_LightSwitcher_C
  tracks [{
    type=MovieSceneEventTrack
    sections [{
      type=MovieSceneEventTriggerSection
      channels [None {type=MovieSceneScriptingEventChannel, const ["<Struct 'MovieSceneEvent' (0x8519a7f4a60) {}>", 1]}]
     }]
   }]
 }
 Doors {
  id=#6
  parent_id=#7
  possessed_class=StaticMeshComponent
  tracks [
   {
    type=MovieSceneComponentMaterialTrack
    display_name=Material Element 0
    sections [{type=MovieSceneParameterSection, channels [Open {type=MovieSceneScriptingFloatChannel, default=1, keys [[20,0],[29,1]]}]}]
   }
   {
    type=MovieSceneComponentMaterialTrack
    display_name=Material Element 0
    sections [{type=MovieSceneParameterSection, channels [Open {type=MovieSceneScriptingFloatChannel, default=1, keys [[20,0],[29,1]]}]}]
   }
  ]
 }
 BP_Elevator_Transition {
  id=#7
  possessed_class=BP_Elevator_Transition_C
  tracks [{type=MovieSceneEventTrack, sections [{type=MovieSceneEventTriggerSection, channels [None {type=MovieSceneScriptingEventChannel}]}]}]
 }
 BP_Battlepass_LightSwitcher_Maproom {
  id=#8
  possessed_class=BP_Battlepass_LightSwitcher_C
  tracks [{
    type=MovieSceneEventTrack
    sections [{
      type=MovieSceneEventTriggerSection
      channels [None {type=MovieSceneScriptingEventChannel, const ["<Struct 'MovieSceneEvent' (0x8519a7f42e0) {}>", 1]}]
     }]
   }]
 }
 BP_Elevator_Transition {
  id=#2
  possessed_class=BP_Elevator_Transition_C
  tracks [{
    type=MovieSceneEventTrack
    sections [{
      type=MovieSceneEventTriggerSection
      channels [None {
        type=MovieSceneScriptingEventChannel
        keys [[1, "<Struct 'MovieSceneEvent' (0x8519a7f3840) {}>"], [11, "<Struct 'MovieSceneEvent' (0x8519a7f3660) {}>"]]
       }]
     }]
   }]
 }
 Camera_Root_Movement {
  id=#4
  possessed_class=Frontend_EventLevel_Camera_C
  tracks [{
    type=MovieScene3DTransformTrack
    sections [{
      type=MovieScene3DTransformSection
      channels [
       Location.X {type=MovieSceneScriptingDoubleChannel, default=2755.2795, keys [[0,637.1765],[5,637.1765],[15,3054.3296],[20,3054.3296],[40,2755.2795]]}
       Location.Y {
        type=MovieSceneScriptingDoubleChannel
        default=16400.7832
        keys [[0,-1200.7053],[5,-1200.7053],[15,16211.2393],[20,16211.2393],[40,16400.7832]]
       }
       Location.Z {type=MovieSceneScriptingDoubleChannel, default=986.691, keys [[0,551.1627],[5,551.1627],[15,1016.4581],[20,1016.4581],[40,986.691]]}
       Rotation.X {type=MovieSceneScriptingDoubleChannel, default=0.0001, thinned_from=5, keys [[0,0],[40,0.0001]]}
       Rotation.Y {type=MovieSceneScriptingDoubleChannel, default=-1.6, thinned_from=5, keys [[0,0],[20,0],[40,-1.6]]}
       Rotation.Z {type=MovieSceneScriptingDoubleChannel, default=138.2947, keys [[0,-196.9],[5,-196.9],[15,158.4375],[20,158.4375],[40,138.2947]]}
       Scale.X {type=MovieSceneScriptingDoubleChannel, default=1}
       Scale.Y {type=MovieSceneScriptingDoubleChannel, default=1}
       Scale.Z {type=MovieSceneScriptingDoubleChannel, default=1}
      ]
     }]
   }]
 }
]
display_rate {denominator=1, numerator=30}
tick_resolution {denominator=1, numerator=24000}
tracks [
 {
  type=MovieSceneEventTrack
  sections [{
    type=MovieSceneEventTriggerSection
    channels [None {type=MovieSceneScriptingEventChannel, const ["<Struct 'MovieSceneEvent' (0x8519a7f4560) {}>", 1]}]
   }]
 }
 {
  type=MovieSceneEventTrack
  sections [{
    type=MovieSceneEventTriggerSection
    channels [None {type=MovieSceneScriptingEventChannel, const ["<Struct 'MovieSceneEvent' (0x8519a7f4100) {}>", 1]}]
   }]
 }
]
