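# Front-end level sequence that moves the camera into the second inspect view of the S12 CatBurglar faction select screen, as its name indicates.
# A transform track on the possessed Frontend_EventLevel_Camera actor makes a short two-key move, mostly a yaw swing with a slight reposition.
# Its child CameraComponent holds a fixed field of view.
# Two sequence-level event trigger tracks fire events whose targets are not shown.
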LevelSequence /MiscReferencersOfCosmetics/Environments/FrontEnd/LevelSequence/BP_12/S12_CatBurglar_FactionSelect_ToInspect02
bindings [
 CameraComponent {
  id=#1
  parent_id=#2
  possessed_class=CameraComponent
  tracks [{
    type=MovieSceneFloatTrack
    display_name=Field Of View
    sections [{type=MovieSceneFloatSection, channels [None {type=MovieSceneScriptingFloatChannel, default=80, const [80,2]}]}]
   }]
 }
 Camera_Root_Movement {
  id=#2
  possessed_class=Frontend_EventLevel_Camera_C
  tracks [{
    type=MovieScene3DTransformTrack
    sections [{
      type=MovieScene3DTransformSection
      channels [
       Location.X {type=MovieSceneScriptingDoubleChannel, default=2729.7156, keys [[0,2786.0232],[15,2729.7156]]}
       Location.Y {type=MovieSceneScriptingDoubleChannel, default=-119966.8906, keys [[0,-120067.125],[15,-119966.8906]]}
       Location.Z {type=MovieSceneScriptingDoubleChannel, default=93.5839, keys [[0,91.7045],[15,93.5839]]}
       Rotation.X {type=MovieSceneScriptingDoubleChannel, default=-0.6842, keys [[0,0],[15,-0.6842]]}
       Rotation.Y {type=MovieSceneScriptingDoubleChannel, default=1.8793, keys [[0,2],[15,1.8793]]}
       Rotation.Z {type=MovieSceneScriptingDoubleChannel, default=-157.9865, keys [[0,-175],[15,-157.9865]]}
       Scale.X {type=MovieSceneScriptingDoubleChannel, default=1}
       Scale.Y {type=MovieSceneScriptingDoubleChannel, default=1}
       Scale.Z {type=MovieSceneScriptingDoubleChannel, default=1}
      ]
     }]
   }]
 }
]
display_rate {denominator=1, numerator=30}
tick_resolution {denominator=1, numerator=24000}
tracks [
 {
  type=MovieSceneEventTrack
  sections [{
    type=MovieSceneEventTriggerSection
    channels [None {type=MovieSceneScriptingEventChannel, const ["<Struct 'MovieSceneEvent' (0x8519a7f46a0) {}>", 1]}]
   }]
 }
 {
  type=MovieSceneEventTrack
  sections [{
    type=MovieSceneEventTriggerSection
    channels [None {type=MovieSceneScriptingEventChannel, const ["<Struct 'MovieSceneEvent' (0x8519a7f42e0) {}>", 1]}]
   }]
 }
]
