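# Level sequence for the S12 front end that, by its name, moves the camera from the CatBurglar view to faction select.
# A short transform track pans and moves the Frontend_EventLevel_Camera root, with a slight roll that returns to level.
# Its child CameraComponent holds a constant field of view.
# Two event trigger tracks each fire one event, whose target is not shown.
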LevelSequence /MiscReferencersOfCosmetics/Environments/FrontEnd/LevelSequence/BP_12/S12_CatBurglar_ToFactionSelect
bindings [
 CameraComponent {
  id=#1
  parent_id=#2
  possessed_class=CameraComponent
  tracks [{
    type=MovieSceneFloatTrack
    display_name=Field Of View
    sections [{type=MovieSceneFloatSection, channels [None {type=MovieSceneScriptingFloatChannel, default=80, const [80,2]}]}]
   }]
 }
 Camera_Root_Movement {
  id=#2
  possessed_class=Frontend_EventLevel_Camera_C
  tracks [{
    type=MovieScene3DTransformTrack
    sections [{
      type=MovieScene3DTransformSection
      channels [
       Location.X {type=MovieSceneScriptingDoubleChannel, default=2803.4331, keys [[0,2803.4331],[15,2786.0232]]}
       Location.Y {type=MovieSceneScriptingDoubleChannel, default=-120226.7109, keys [[0,-120226.7109],[15,-120067.125]]}
       Location.Z {type=MovieSceneScriptingDoubleChannel, default=107.8428, keys [[0,107.8428],[15,91.7045]]}
       Rotation.X {type=MovieSceneScriptingDoubleChannel, default=-2, keys [[0,0],[8,-1.5],[15,0]]}
       Rotation.Y {type=MovieSceneScriptingDoubleChannel, default=1.7888, keys [[0,-2.4004],[15,2]]}
       Rotation.Z {type=MovieSceneScriptingDoubleChannel, default=-175, keys [[0,-214.4646],[15,-175]]}
       Scale.X {type=MovieSceneScriptingDoubleChannel, default=1}
       Scale.Y {type=MovieSceneScriptingDoubleChannel, default=1}
       Scale.Z {type=MovieSceneScriptingDoubleChannel, default=1}
      ]
     }]
   }]
 }
]
display_rate {denominator=1, numerator=30}
tick_resolution {denominator=1, numerator=24000}
tracks [
 {
  type=MovieSceneEventTrack
  sections [{
    type=MovieSceneEventTriggerSection
    channels [None {type=MovieSceneScriptingEventChannel, const ["<Struct 'MovieSceneEvent' (0x8519a7f4600) {}>", 1]}]
   }]
 }
 {
  type=MovieSceneEventTrack
  sections [{
    type=MovieSceneEventTriggerSection
    channels [None {type=MovieSceneScriptingEventChannel, const ["<Struct 'MovieSceneEvent' (0x8519a7f4a60) {}>", 1]}]
   }]
 }
]
