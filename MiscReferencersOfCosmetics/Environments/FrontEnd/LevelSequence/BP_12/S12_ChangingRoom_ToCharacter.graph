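# Front-end Level Sequence for a short camera move in the changing room, apparently transitioning to a character view, as its name suggests.
# It animates a possessed Frontend_EventLevel_Camera actor over ten keyed frames, shifting and raising it while turning and pitching it down.
# The child CameraComponent narrows its field of view over the same span to zoom in.
# A root event trigger track holds an event whose target is not shown.
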LevelSequence /MiscReferencersOfCosmetics/Environments/FrontEnd/LevelSequence/BP_12/S12_ChangingRoom_ToCharacter
bindings [
 Camera_Root_Movement {
  id=#1
  possessed_class=Frontend_EventLevel_Camera_C
  tracks [{
    type=MovieScene3DTransformTrack
    sections [{
      type=MovieScene3DTransformSection
      channels [
       Location.X {type=MovieSceneScriptingDoubleChannel, default=2264.3303, keys [[0,2264.3303],[10,2269.2996]]}
       Location.Y {type=MovieSceneScriptingDoubleChannel, default=6327.0493, keys [[0,6327.0493],[10,6290.3208]]}
       Location.Z {type=MovieSceneScriptingDoubleChannel, default=899.3929, keys [[0,899.3929],[10,958.3929]]}
       Rotation.X {type=MovieSceneScriptingDoubleChannel, default=0, keys [[0,0],[10,0]]}
       Rotation.Y {type=MovieSceneScriptingDoubleChannel, default=-3.1963, keys [[0,-3.1963],[10,-15.3997]]}
       Rotation.Z {type=MovieSceneScriptingDoubleChannel, default=139.195, keys [[0,139.195],[10,134.5939]]}
       Scale.X {type=MovieSceneScriptingDoubleChannel, default=1}
       Scale.Y {type=MovieSceneScriptingDoubleChannel, default=1}
       Scale.Z {type=MovieSceneScriptingDoubleChannel, default=1}
      ]
     }]
   }]
 }
 CameraComponent {
  id=#2
  parent_id=#1
  possessed_class=CameraComponent
  tracks [{
    type=MovieSceneFloatTrack
    display_name=Field Of View
    sections [{type=MovieSceneFloatSection, channels [None {type=MovieSceneScriptingFloatChannel, default=75, keys [[0,75],[10,60]]}]}]
   }]
 }
]
display_rate {denominator=1, numerator=30}
tick_resolution {denominator=1, numerator=24000}
tracks [{
  type=MovieSceneEventTrack
  sections [{
    type=MovieSceneEventTriggerSection
    channels [None {type=MovieSceneScriptingEventChannel, const ["<Struct 'MovieSceneEvent' (0x8519a7f4c40) {}>", 1]}]
   }]
 }]
