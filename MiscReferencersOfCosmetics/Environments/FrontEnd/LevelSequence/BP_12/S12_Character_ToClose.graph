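# Front-end level sequence in the BP_12 set that, going by its name, transitions the camera from a character view to a close framing.
# It animates a possessed Frontend_EventLevel_Camera actor through a short location and rotation move over keys 0 to 15.
# The child CameraComponent's field of view narrows over the same keys, adding a zoom to the move.
# Two event trigger tracks run alongside, though the events they call are not identified.
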LevelSequence /MiscReferencersOfCosmetics/Environments/FrontEnd/LevelSequence/BP_12/S12_Character_ToClose
bindings [
 CameraComponent {
  id=#1
  parent_id=#2
  possessed_class=CameraComponent
  tracks [{
    type=MovieSceneFloatTrack
    display_name=Field Of View
    sections [{type=MovieSceneFloatSection, channels [None {type=MovieSceneScriptingFloatChannel, default=50, keys [[0,60],[15,50]]}]}]
   }]
 }
 Camera_Root_Movement {
  id=#2
  possessed_class=Frontend_EventLevel_Camera_C
  tracks [{
    type=MovieScene3DTransformTrack
    sections [{
      type=MovieScene3DTransformSection
      channels [
       Location.X {type=MovieSceneScriptingDoubleChannel, default=2093.8115, keys [[0,2269.2996],[15,2127.4451]]}
       Location.Y {type=MovieSceneScriptingDoubleChannel, default=6365.8657, keys [[0,6290.3208],[15,6358.8345]]}
       Location.Z {type=MovieSceneScriptingDoubleChannel, default=820.8658, keys [[0,958.3929],[15,937.3929]]}
       Rotation.X {type=MovieSceneScriptingDoubleChannel, default=0, keys [[0,0],[15,0]]}
       Rotation.Y {type=MovieSceneScriptingDoubleChannel, default=-3.9382, keys [[0,-15.3997],[15,-8.5998]]}
       Rotation.Z {type=MovieSceneScriptingDoubleChannel, default=116.9666, keys [[0,134.5939],[15,129.9919]]}
       Scale.X {type=MovieSceneScriptingDoubleChannel, default=1}
       Scale.Y {type=MovieSceneScriptingDoubleChannel, default=1}
       Scale.Z {type=MovieSceneScriptingDoubleChannel, default=1}
      ]
     }]
   }]
 }
]
display_rate {denominator=1, numerator=30}
tick_resolution {denominator=1, numerator=24000}
tracks [
 {
  type=MovieSceneEventTrack
  sections [{
    type=MovieSceneEventTriggerSection
    channels [None {type=MovieSceneScriptingEventChannel, const ["<Struct 'MovieSceneEvent' (0x8519a7f4560) {}>", 1]}]
   }]
 }
 {
  type=MovieSceneEventTrack
  sections [{
    type=MovieSceneEventTriggerSection
    channels [None {type=MovieSceneScriptingEventChannel, const ["<Struct 'MovieSceneEvent' (0x8519a7f4060) {}>", 1]}]
   }]
 }
]
