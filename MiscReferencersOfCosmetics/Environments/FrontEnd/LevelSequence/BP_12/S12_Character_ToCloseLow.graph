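# Level Sequence for the front end that moves the Frontend_EventLevel_Camera to a close, low character framing, likely for Battle Pass 12 judging by its path.
# Over 15 keyed frames at 30 fps, the camera root descends, shifts position, flattens its downward pitch and swings its yaw.
# The child CameraComponent holds a constant field of view.
# Two event tracks each carry a trigger whose target function is not shown.
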LevelSequence /MiscReferencersOfCosmetics/Environments/FrontEnd/LevelSequence/BP_12/S12_Character_ToCloseLow
bindings [
 CameraComponent {
  id=#1
  parent_id=#2
  possessed_class=CameraComponent
  tracks [{
    type=MovieSceneFloatTrack
    display_name=Field Of View
    sections [{type=MovieSceneFloatSection, channels [None {type=MovieSceneScriptingFloatChannel, default=60, const [60,2]}]}]
   }]
 }
 Camera_Root_Movement {
  id=#2
  possessed_class=Frontend_EventLevel_Camera_C
  tracks [{
    type=MovieScene3DTransformTrack
    sections [{
      type=MovieScene3DTransformSection
      channels [
       Location.X {type=MovieSceneScriptingDoubleChannel, default=2114.7891, keys [[0,2269.2996],[15,2114.7891]]}
       Location.Y {type=MovieSceneScriptingDoubleChannel, default=6324.6377, keys [[0,6290.3208],[15,6324.6377]]}
       Location.Z {type=MovieSceneScriptingDoubleChannel, default=840.263, keys [[0,958.3929],[15,840.263]]}
       Rotation.X {type=MovieSceneScriptingDoubleChannel, default=0, keys [[0,0],[14.9987,0]]}
       Rotation.Y {type=MovieSceneScriptingDoubleChannel, default=-3, keys [[0,-15.3997],[15,-3]]}
       Rotation.Z {type=MovieSceneScriptingDoubleChannel, default=116.9666, keys [[0,134.5939],[15,116.9666]]}
       Scale.X {type=MovieSceneScriptingDoubleChannel, default=1}
       Scale.Y {type=MovieSceneScriptingDoubleChannel, default=1}
       Scale.Z {type=MovieSceneScriptingDoubleChannel, default=1}
      ]
     }]
   }]
 }
]
display_rate {denominator=1, numerator=30}
tick_resolution {denominator=1, numerator=24000}
tracks [
 {
  type=MovieSceneEventTrack
  sections [{
    type=MovieSceneEventTriggerSection
    channels [None {type=MovieSceneScriptingEventChannel, const ["<Struct 'MovieSceneEvent' (0x8519a7f3840) {}>", 1]}]
   }]
 }
 {
  type=MovieSceneEventTrack
  sections [{
    type=MovieSceneEventTriggerSection
    channels [None {type=MovieSceneScriptingEventChannel, const ["<Struct 'MovieSceneEvent' (0x8519a7f4d80) {}>", 1]}]
   }]
 }
]
