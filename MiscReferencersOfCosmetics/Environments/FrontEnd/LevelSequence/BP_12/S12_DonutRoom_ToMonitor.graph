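# Level sequence for the Battle Pass 12 front-end that moves the camera from the donut room to a close-up of the Agency computer monitor.
# The event-level camera root is keyed down and around toward the computer over about 25 frames at 30 fps with a fixed field of view.
# The computer's material MinimizeAnim parameter holds, then drops to near zero late in the move.
# Two event trigger tracks are also present.
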LevelSequence /MiscReferencersOfCosmetics/Environments/FrontEnd/LevelSequence/BP_12/S12_DonutRoom_ToMonitor
bindings [
 S_Agency_Computer_01 {
  id=#1
  possessed_class=FortStaticMeshActor
  tracks [{
    type=MovieScene3DTransformTrack
    sections [{
      type=MovieScene3DTransformSection
      channels [
       Location.X {type=MovieSceneScriptingDoubleChannel, default=-1208}
       Location.Y {type=MovieSceneScriptingDoubleChannel, default=-8680}
       Location.Z {type=MovieSceneScriptingDoubleChannel, default=112}
       Rotation.X {type=MovieSceneScriptingDoubleChannel, default=-10.1145}
       Rotation.Y {type=MovieSceneScriptingDoubleChannel, default=-9.9233}
       Rotation.Z {type=MovieSceneScriptingDoubleChannel, default=0.8803}
       Scale.X {type=MovieSceneScriptingDoubleChannel, default=0.7427}
       Scale.Y {type=MovieSceneScriptingDoubleChannel, default=0.7427}
       Scale.Z {type=MovieSceneScriptingDoubleChannel, default=0.7427}
      ]
     }]
   }]
 }
 StaticMeshComponent0 {
  id=#2
  parent_id=#1
  possessed_class=StaticMeshComponent
  tracks [{
    type=MovieSceneComponentMaterialTrack
    display_name=Material Element 0
    sections [{
      type=MovieSceneParameterSection
      end_frame=26
      start_frame=0
      channels [MinimizeAnim {type=MovieSceneScriptingFloatChannel, default=0.0001, keys [[0,1],[15.08,1],[22.405,0.0001]]}]
     }]
   }]
 }
 CameraComponent {
  id=#3
  parent_id=#4
  possessed_class=CameraComponent
  tracks [{
    type=MovieSceneFloatTrack
    display_name=Field Of View
    sections [{type=MovieSceneFloatSection, channels [None {type=MovieSceneScriptingFloatChannel, default=75, const [75,1]}]}]
   }]
 }
 Camera_Root_Movement {
  id=#4
  possessed_class=Frontend_EventLevel_Camera_C
  tracks [{
    type=MovieScene3DTransformTrack
    sections [{
      type=MovieScene3DTransformSection
      channels [
       Location.X {type=MovieSceneScriptingDoubleChannel, default=-852.7941, keys [[0,-852.7941],[11,-1056.5531],[25,-1209.2333]]}
       Location.Y {type=MovieSceneScriptingDoubleChannel, default=-8562.3223, keys [[0,-8562.3223],[17,-8613.9199],[25,-8665.5703]]}
       Location.Z {type=MovieSceneScriptingDoubleChannel, default=365.527, keys [[0,365.527],[8,291.9569],[20,159.138],[25,147.005]]}
       Rotation.X {type=MovieSceneScriptingDoubleChannel, default=0, keys [[0,0],[8,5.4924],[25,11.0026]]}
       Rotation.Y {type=MovieSceneScriptingDoubleChannel, default=-26.2002, keys [[0,-26.2002],[13,-24.3393],[20,-21.6724],[25,-21.5911]]}
       Rotation.Z {type=MovieSceneScriptingDoubleChannel, default=-168.1986, keys [[0,-168.1986],[17,-137.3652],[25,-93.8631]]}
       Scale.X {type=MovieSceneScriptingDoubleChannel, default=1}
       Scale.Y {type=MovieSceneScriptingDoubleChannel, default=1}
       Scale.Z {type=MovieSceneScriptingDoubleChannel, default=1}
      ]
     }]
   }]
 }
]
display_rate {denominator=1, numerator=30}
tick_resolution {denominator=1, numerator=24000}
tracks [
 {
  type=MovieSceneEventTrack
  sections [{
    type=MovieSceneEventTriggerSection
    channels [None {type=MovieSceneScriptingEventChannel, const ["<Struct 'MovieSceneEvent' (0x8519a7f3840) {}>", 1]}]
   }]
 }
 {
  type=MovieSceneEventTrack
  sections [{
    type=MovieSceneEventTriggerSection
    channels [None {type=MovieSceneScriptingEventChannel, const ["<Struct 'MovieSceneEvent' (0x8519a7f4920) {}>", 1]}]
   }]
 }
]
