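# Front-end camera transition sequence in the BP_12 set that, by its name, moves the view from the Donut Room to a note.
# It animates the root transform of a Frontend_EventLevel_Camera, lowering it and pitching it steeply downward with slight roll and yaw over less than a second.
# The child CameraComponent widens its field of view over the same span.
# It contains only these camera tracks, with no master tracks.
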LevelSequence /MiscReferencersOfCosmetics/Environments/FrontEnd/LevelSequence/BP_12/S12_DonutRoom_ToNote
bindings [
 CameraComponent {
  id=#1
  parent_id=#2
  possessed_class=CameraComponent
  tracks [{
    type=MovieSceneFloatTrack
    display_name=Field Of View
    sections [{type=MovieSceneFloatSection, channels [None {type=MovieSceneScriptingFloatChannel, default=90, keys [[0,75],[25,90]]}]}]
   }]
 }
 Camera_Root_Movement {
  id=#2
  possessed_class=Frontend_EventLevel_Camera_C
  tracks [{
    type=MovieScene3DTransformTrack
    sections [{
      type=MovieScene3DTransformSection
      channels [
       Location.X {type=MovieSceneScriptingDoubleChannel, default=-1198, keys [[0,-852.7941],[25,-1198]]}
       Location.Y {type=MovieSceneScriptingDoubleChannel, default=-8514.3223, keys [[0,-8562.3223],[25,-8514.3223]]}
       Location.Z {type=MovieSceneScriptingDoubleChannel, default=41.527, keys [[0,365.527],[25,41.527]]}
       Rotation.X {type=MovieSceneScriptingDoubleChannel, default=-5, keys [[0,0],[12,-0.461],[25,-5]]}
       Rotation.Y {type=MovieSceneScriptingDoubleChannel, default=-60, keys [[0,-26.2002],[11,-36.9107],[25,-60]]}
       Rotation.Z {type=MovieSceneScriptingDoubleChannel, default=-176, keys [[0,-168.1986],[25,-176]]}
       Scale.X {type=MovieSceneScriptingDoubleChannel, default=1}
       Scale.Y {type=MovieSceneScriptingDoubleChannel, default=1}
       Scale.Z {type=MovieSceneScriptingDoubleChannel, default=1}
      ]
     }]
   }]
 }
]
display_rate {denominator=1, numerator=30}
tick_resolution {denominator=1, numerator=24000}
tracks []
