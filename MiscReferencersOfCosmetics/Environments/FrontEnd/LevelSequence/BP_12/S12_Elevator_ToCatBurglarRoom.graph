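# Front-end Level Sequence for the Season 12 Battle Pass screen that moves the camera from the elevator to the Cat Burglar room in about 40 frames.
# The camera travels to a distant location, holds, then eases into its final framing with a field-of-view settle.
# Elevator doors open through an animated "Open" material parameter.
# Event tracks trigger the elevator transition, Midas and Maproom light switchers, camera stream-in/out starts, and two sequence-level events.
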LevelSequence /MiscReferencersOfCosmetics/Environments/FrontEnd/LevelSequence/BP_12/S12_Elevator_ToCatBurglarRoom
bindings [
 Doors {id=#1, parent_id=#2, possessed_class=StaticMeshComponent, tracks []}
 CameraComponent {
  id=#3
  parent_id=#4
  possessed_class=CameraComponent
  tracks [{
    type=MovieSceneFloatTrack
    display_name=Field Of View
    sections [{type=MovieSceneFloatSection, channels [None {type=MovieSceneScriptingFloatChannel, default=75, keys [[20,80],[40,75]]}]}]
   }]
 }
 BP_Battlepass_LightSwitcher_Midas {
  id=#5
  possessed_class=BP_Battlepass_LightSwitcher_C
  tracks [{
    type=MovieSceneEventTrack
    sections [{
      type=MovieSceneEventTriggerSection
      channels [None {type=MovieSceneScriptingEventChannel, const ["<Struct 'MovieSceneEvent' (0x8519a7f5d20) {}>", 1]}]
     }]
   }]
 }
 Doors {
  id=#6
  parent_id=#7
  possessed_class=StaticMeshComponent
  tracks [
   {
    type=MovieSceneComponentMaterialTrack
    display_name=Material Element 0
    sections [{type=MovieSceneParameterSection, channels [Open {type=MovieSceneScriptingFloatChannel, default=0, keys [[15,0],[20,0],[29,1]]}]}]
   }
   {
    type=MovieSceneComponentMaterialTrack
    display_name=Material Element 0
    sections [{type=MovieSceneParameterSection, channels [Open {type=MovieSceneScriptingFloatChannel, default=0, keys [[15,0],[20,0],[29,1]]}]}]
   }
  ]
 }
 BP_Elevator_Transition {id=#7, possessed_class=BP_Elevator_Transition_C, tracks []}
 BP_Battlepass_LightSwitcher_Maproom {
  id=#8
  possessed_class=BP_Battlepass_LightSwitcher_C
  tracks [{
    type=MovieSceneEventTrack
    sections [{
      type=MovieSceneEventTriggerSection
      channels [None {type=MovieSceneScriptingEventChannel, const ["<Struct 'MovieSceneEvent' (0x8519a7f5a00) {}>", 1]}]
     }]
   }]
 }
 BP_Elevator_Transition {
  id=#2
  possessed_class=BP_Elevator_Transition_C
  tracks [{
    type=MovieSceneEventTrack
    sections [{
      type=MovieSceneEventTriggerSection
      channels [None {
        type=MovieSceneScriptingEventChannel
        keys [[1, "<Struct 'MovieSceneEvent' (0x8519a7f5be0) {}>"], [11, "<Struct 'MovieSceneEvent' (0x8519a7f5960) {}>"]]
       }]
     }]
   }]
 }
 Camera_Root_Movement {
  id=#4
  possessed_class=Frontend_EventLevel_Camera_C
  tracks [
   {
    type=MovieScene3DTransformTrack
    sections [{
      type=MovieScene3DTransformSection
      channels [
       Location.X {type=MovieSceneScriptingDoubleChannel, default=2803.4331, keys [[0,637.1765],[5,637.1765],[15,3487.7178],[20,3487.7178],[40,2803.4331]]}
       Location.Y {
        type=MovieSceneScriptingDoubleChannel
        default=-120226.7109
        thinned_from=5
        keys [[0,-1200.7053],[5,-1200.7053],[15,-120469.9219],[40,-120226.7109]]
       }
       Location.Z {type=MovieSceneScriptingDoubleChannel, default=107.8428, keys [[0,551.1627],[5,551.1627],[15,155.7106],[20,155.7106],[40,107.8428]]}
       Rotation.X {type=MovieSceneScriptingDoubleChannel, default=0, thinned_from=6, keys [[0,0],[20,0],[29,-1],[40,0]]}
       Rotation.Y {type=MovieSceneScriptingDoubleChannel, default=-2.4004, thinned_from=5, keys [[0,0],[20,0],[40,-2.4004]]}
       Rotation.Z {type=MovieSceneScriptingDoubleChannel, default=145.5354, keys [[0,-196.9],[5,-196.9],[15,165.7739],[20,165.7739],[40,145.5354]]}
       Scale.X {type=MovieSceneScriptingDoubleChannel, default=1}
       Scale.Y {type=MovieSceneScriptingDoubleChannel, default=1}
       Scale.Z {type=MovieSceneScriptingDoubleChannel, default=1}
      ]
     }]
   }
   {
    type=MovieSceneEventTrack
    display_name=Event Stream in Start
    sections [{
      type=MovieSceneEventTriggerSection
      channels [None {type=MovieSceneScriptingEventChannel, const ["<Struct 'MovieSceneEvent' (0x8519a7f5640) {}>", 1]}]
     }]
   }
   {
    type=MovieSceneEventTrack
    display_name=Event Stream out Start
    sections [{
      type=MovieSceneEventTriggerSection
      channels [None {type=MovieSceneScriptingEventChannel, const ["<Struct 'MovieSceneEvent' (0x8519a7f5500) {}>", 1]}]
     }]
   }
  ]
 }
]
display_rate {denominator=1, numerator=30}
tick_resolution {denominator=1, numerator=24000}
tracks [
 {
  type=MovieSceneEventTrack
  sections [{
    type=MovieSceneEventTriggerSection
    channels [None {type=MovieSceneScriptingEventChannel, const ["<Struct 'MovieSceneEvent' (0x8519a7f5c80) {}>", 1]}]
   }]
 }
 {
  type=MovieSceneEventTrack
  sections [{
    type=MovieSceneEventTriggerSection
    channels [None {type=MovieSceneScriptingEventChannel, const ["<Struct 'MovieSceneEvent' (0x8519a7f55a0) {}>", 1]}]
   }]
 }
]
