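# Front-end Level Sequence for the Battle Pass lobby's camera transition via elevator to the Henchman room.
# The event-level camera root holds, jumps to a new location, then eases into final framing as field of view shifts.
# One elevator's door material animates an Open parameter while another elevator receives timed events.
# Event triggers also drive the Henchman and Maproom light switchers, camera stream-in and stream-out, and sequence-level events.
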LevelSequence /MiscReferencersOfCosmetics/Environments/FrontEnd/LevelSequence/BP_12/S12_Elevator_ToHenchmanRoom
bindings [
 Doors {
  id=#1
  parent_id=#2
  possessed_class=StaticMeshComponent
  tracks [
   {
    type=MovieSceneComponentMaterialTrack
    display_name=Material Element 0
    sections [{
      type=MovieSceneParameterSection
      end_frame=51
      start_frame=0
      channels [Open {type=MovieSceneScriptingFloatChannel, default=1, keys [[25,0],[36,1]]}]
     }]
   }
   {
    type=MovieSceneComponentMaterialTrack
    display_name=Material Element 0
    sections [{
      type=MovieSceneParameterSection
      end_frame=51
      start_frame=0
      channels [Open {type=MovieSceneScriptingFloatChannel, default=1, keys [[25,0],[36,1]]}]
     }]
   }
  ]
 }
 BP_Elevator_Transition3 {
  id=#2
  possessed_class=BP_Elevator_Transition_C
  tracks [{type=MovieSceneEventTrack, sections [{type=MovieSceneEventTriggerSection, channels [None {type=MovieSceneScriptingEventChannel}]}]}]
 }
 Doors {id=#3, parent_id=#4, possessed_class=StaticMeshComponent, tracks []}
 BP_Elevator_Transition {
  id=#4
  possessed_class=BP_Elevator_Transition_C
  tracks [{
    type=MovieSceneEventTrack
    sections [{
      type=MovieSceneEventTriggerSection
      channels [None {
        type=MovieSceneScriptingEventChannel
        keys [[1, "<Struct 'MovieSceneEvent' (0x8519a7f51e0) {}>"], [11, "<Struct 'MovieSceneEvent' (0x8519a7f5d20) {}>"]]
       }]
     }]
   }]
 }
 CameraComponent {
  id=#5
  parent_id=#6
  possessed_class=CameraComponent
  tracks [{
    type=MovieSceneFloatTrack
    display_name=Field Of View
    sections [{type=MovieSceneFloatSection, channels [None {type=MovieSceneScriptingFloatChannel, default=75, keys [[30,80],[50,75]]}]}]
   }]
 }
 BP_Battlepass_LightSwitcher_Hench {
  id=#7
  possessed_class=BP_Battlepass_LightSwitcher_C
  tracks [{
    type=MovieSceneEventTrack
    sections [{
      type=MovieSceneEventTriggerSection
      channels [None {type=MovieSceneScriptingEventChannel, const ["<Struct 'MovieSceneEvent' (0x8519a7f4d80) {}>", 1]}]
     }]
   }]
 }
 BP_Battlepass_LightSwitcher_Maproom {
  id=#8
  possessed_class=BP_Battlepass_LightSwitcher_C
  tracks [{
    type=MovieSceneEventTrack
    sections [{
      type=MovieSceneEventTriggerSection
      channels [None {type=MovieSceneScriptingEventChannel, const ["<Struct 'MovieSceneEvent' (0x8519a7f56e0) {}>", 1]}]
     }]
   }]
 }
 Camera_Root_Movement {
  id=#6
  possessed_class=Frontend_EventLevel_Camera_C
  tracks [
   {
    type=MovieScene3DTransformTrack
    sections [{
      type=MovieScene3DTransformSection
      channels [
       Location.X {type=MovieSceneScriptingDoubleChannel, default=637.1765, keys [[0,637.1765],[11,637.1765],[25,2358.5],[30,2358.5],[50,2043.6157]]}
       Location.Y {type=MovieSceneScriptingDoubleChannel, default=-1200.7053, keys [[0,-1200.7053],[11,-1200.7053],[25,3984],[30,3984],[50,3752.7812]]}
       Location.Z {type=MovieSceneScriptingDoubleChannel, default=551.1627, keys [[0,551.1627],[11,551.1627],[25,926],[30,926],[50,931.1245]]}
       Rotation.X {type=MovieSceneScriptingDoubleChannel, default=0, thinned_from=5, keys [[0,0],[50,0]]}
       Rotation.Y {type=MovieSceneScriptingDoubleChannel, default=0, thinned_from=5, keys [[0,0],[30,0],[50,-2]]}
       Rotation.Z {type=MovieSceneScriptingDoubleChannel, default=-196.9, keys [[0,-196.9],[11,-196.9],[25,-130],[30,-130],[50,-150]]}
       Scale.X {type=MovieSceneScriptingDoubleChannel, default=1}
       Scale.Y {type=MovieSceneScriptingDoubleChannel, default=1}
       Scale.Z {type=MovieSceneScriptingDoubleChannel, default=1}
      ]
     }]
   }
   {
    type=MovieSceneEventTrack
    display_name=Event Stream in Start
    sections [{
      type=MovieSceneEventTriggerSection
      channels [None {type=MovieSceneScriptingEventChannel, const ["<Struct 'MovieSceneEvent' (0x8519a7f4f60) {}>", 1]}]
     }]
   }
   {
    type=MovieSceneEventTrack
    display_name=Event Stream out Start
    sections [{
      type=MovieSceneEventTriggerSection
      channels [None {type=MovieSceneScriptingEventChannel, const ["<Struct 'MovieSceneEvent' (0x8519a7f4100) {}>", 1]}]
     }]
   }
  ]
 }
]
display_rate {denominator=1, numerator=30}
tick_resolution {denominator=1, numerator=24000}
tracks [
 {
  type=MovieSceneEventTrack
  sections [{
    type=MovieSceneEventTriggerSection
    channels [None {type=MovieSceneScriptingEventChannel, const ["<Struct 'MovieSceneEvent' (0x8519a7f5c80) {}>", 1]}]
   }]
 }
 {
  type=MovieSceneEventTrack
  sections [{
    type=MovieSceneEventTriggerSection
    channels [None {type=MovieSceneScriptingEventChannel, const ["<Struct 'MovieSceneEvent' (0x8519a7f5320) {}>", 1]}]
   }]
 }
]
