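# Front-end level sequence for a short camera transition by elevator to Tina's room in the Battle Pass environment.
# Camera_Root_Movement travels in two keyed moves separated by a hold, at a constant field of view.
# Two elevator transition Blueprints carry event tracks, and one animates its doors through an "Open" material parameter during the final move.
# Event triggers also drive the TNTINA and Maproom light switchers, stream-in/out starts and two sequence-level tracks.
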LevelSequence /MiscReferencersOfCosmetics/Environments/FrontEnd/LevelSequence/BP_12/S12_Elevator_ToTinaRoom
bindings [
 Doors {id=#1, parent_id=#2, possessed_class=StaticMeshComponent, tracks []}
 CameraComponent {
  id=#3
  parent_id=#4
  possessed_class=CameraComponent
  tracks [{
    type=MovieSceneFloatTrack
    display_name=Field Of View
    sections [{type=MovieSceneFloatSection, channels [None {type=MovieSceneScriptingFloatChannel, default=80, const [80,2]}]}]
   }]
 }
 BP_Battlepass_LightSwitcher_TNTINA {
  id=#5
  possessed_class=BP_Battlepass_LightSwitcher_C
  tracks [{
    type=MovieSceneEventTrack
    sections [{
      type=MovieSceneEventTriggerSection
      channels [None {type=MovieSceneScriptingEventChannel, const ["<Struct 'MovieSceneEvent' (0x8519a7f50a0) {}>", 1]}]
     }]
   }]
 }
 Doors {
  id=#6
  parent_id=#7
  possessed_class=StaticMeshComponent
  tracks [
   {
    type=MovieSceneComponentMaterialTrack
    display_name=Material Element 0
    sections [{type=MovieSceneParameterSection, channels [Open {type=MovieSceneScriptingFloatChannel, default=1, keys [[20,0],[29,1]]}]}]
   }
   {
    type=MovieSceneComponentMaterialTrack
    display_name=Material Element 0
    sections [{type=MovieSceneParameterSection, channels [Open {type=MovieSceneScriptingFloatChannel, default=1, keys [[20,0],[29,1]]}]}]
   }
  ]
 }
 BP_Elevator_Transition {
  id=#7
  possessed_class=BP_Elevator_Transition_C
  tracks [{type=MovieSceneEventTrack, sections [{type=MovieSceneEventTriggerSection, channels [None {type=MovieSceneScriptingEventChannel}]}]}]
 }
 BP_Battlepass_LightSwitcher_Maproom {
  id=#8
  possessed_class=BP_Battlepass_LightSwitcher_C
  tracks [{
    type=MovieSceneEventTrack
    sections [{
      type=MovieSceneEventTriggerSection
      channels [None {type=MovieSceneScriptingEventChannel, const ["<Struct 'MovieSceneEvent' (0x8519a7f5aa0) {}>", 1]}]
     }]
   }]
 }
 BP_Elevator_Transition {
  id=#2
  possessed_class=BP_Elevator_Transition_C
  tracks [{
    type=MovieSceneEventTrack
    sections [{
      type=MovieSceneEventTriggerSection
      channels [None {
        type=MovieSceneScriptingEventChannel
        keys [[1, "<Struct 'MovieSceneEvent' (0x8519a7f5780) {}>"], [11, "<Struct 'MovieSceneEvent' (0x8519a7f5500) {}>"]]
       }]
     }]
   }]
 }
 Camera_Root_Movement {
  id=#4
  possessed_class=Frontend_EventLevel_Camera_C
  tracks [
   {
    type=MovieScene3DTransformTrack
    sections [{
      type=MovieScene3DTransformSection
      channels [
       Location.X {type=MovieSceneScriptingDoubleChannel, default=2378.3931, keys [[0,637.1765],[5,637.1765],[15,2700],[20,2700],[40,2378.3931]]}
       Location.Y {type=MovieSceneScriptingDoubleChannel, default=10649.8271, keys [[0,-1200.7053],[5,-1200.7053],[15,10412],[20,10412],[40,10649.8271]]}
       Location.Z {type=MovieSceneScriptingDoubleChannel, default=876.3943, keys [[0,551.1627],[5,551.1627],[15,927],[20,927],[40,876.3943]]}
       Rotation.X {type=MovieSceneScriptingDoubleChannel, default=0, thinned_from=5, keys [[0,0],[40,0]]}
       Rotation.Y {type=MovieSceneScriptingDoubleChannel, default=0.5995, thinned_from=5, keys [[0,0],[20,0],[40,0.5995]]}
       Rotation.Z {type=MovieSceneScriptingDoubleChannel, default=131.5919, keys [[0,-196.9],[5,-196.9],[15,155.71],[20,155.71],[40,131.5919]]}
       Scale.X {type=MovieSceneScriptingDoubleChannel, default=1}
       Scale.Y {type=MovieSceneScriptingDoubleChannel, default=1}
       Scale.Z {type=MovieSceneScriptingDoubleChannel, default=1}
      ]
     }]
   }
   {
    type=MovieSceneEventTrack
    display_name=Event Stream in Start
    sections [{
      type=MovieSceneEventTriggerSection
      channels [None {type=MovieSceneScriptingEventChannel, const ["<Struct 'MovieSceneEvent' (0x8519a7f5be0) {}>", 1]}]
     }]
   }
   {
    type=MovieSceneEventTrack
    display_name=Event Stream out Start
    sections [{
      type=MovieSceneEventTriggerSection
      channels [None {type=MovieSceneScriptingEventChannel, const ["<Struct 'MovieSceneEvent' (0x8519a7f5960) {}>", 1]}]
     }]
   }
  ]
 }
]
display_rate {denominator=1, numerator=30}
tick_resolution {denominator=1, numerator=24000}
tracks [
 {
  type=MovieSceneEventTrack
  sections [{
    type=MovieSceneEventTriggerSection
    channels [None {type=MovieSceneScriptingEventChannel, const ["<Struct 'MovieSceneEvent' (0x8519a7f4d80) {}>", 1]}]
   }]
 }
 {
  type=MovieSceneEventTrack
  sections [{
    type=MovieSceneEventTriggerSection
    channels [None {type=MovieSceneScriptingEventChannel, const ["<Struct 'MovieSceneEvent' (0x8519a7f58c0) {}>", 1]}]
   }]
 }
]
