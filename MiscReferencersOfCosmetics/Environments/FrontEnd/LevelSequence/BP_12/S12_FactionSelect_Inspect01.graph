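# Front-end Level Sequence that, judging by its name, moves the camera into an inspect view on the faction select screen.
# It possesses the Frontend_EventLevel_Camera actor and keys its transform over 15 frames at 30 fps, a short half-second reposition with slight downward pitch and yaw change.
# A child CameraComponent binding holds the field of view constant.
# The sequence has no camera cut or other root tracks.
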LevelSequence /MiscReferencersOfCosmetics/Environments/FrontEnd/LevelSequence/BP_12/S12_FactionSelect_Inspect01
bindings [
 CameraComponent {
  id=#1
  parent_id=#2
  possessed_class=CameraComponent
  tracks [{
    type=MovieSceneFloatTrack
    display_name=Field Of View
    sections [{type=MovieSceneFloatSection, channels [None {type=MovieSceneScriptingFloatChannel, default=75, const [75,1]}]}]
   }]
 }
 Camera_Root_Movement {
  id=#2
  possessed_class=Frontend_EventLevel_Camera_C
  tracks [{
    type=MovieScene3DTransformTrack
    sections [{
      type=MovieScene3DTransformSection
      channels [
       Location.X {type=MovieSceneScriptingDoubleChannel, default=1923.3733, keys [[0,1883.6157],[15,1923.3733]]}
       Location.Y {type=MovieSceneScriptingDoubleChannel, default=3590.6091, keys [[0,3684.7812],[15,3590.6091]]}
       Location.Z {type=MovieSceneScriptingDoubleChannel, default=922.1248, keys [[0,915.1245],[15,922.1248]]}
       Rotation.X {type=MovieSceneScriptingDoubleChannel, default=0, keys [[0,0],[15,0]]}
       Rotation.Y {type=MovieSceneScriptingDoubleChannel, default=-12, keys [[0,-10],[15,-12]]}
       Rotation.Z {type=MovieSceneScriptingDoubleChannel, default=-159, keys [[0,-151.2877],[15,-159]]}
       Scale.X {type=MovieSceneScriptingDoubleChannel, default=1}
       Scale.Y {type=MovieSceneScriptingDoubleChannel, default=1}
       Scale.Z {type=MovieSceneScriptingDoubleChannel, default=1}
      ]
     }]
   }]
 }
]
display_rate {denominator=1, numerator=30}
tick_resolution {denominator=1, numerator=24000}
tracks []
