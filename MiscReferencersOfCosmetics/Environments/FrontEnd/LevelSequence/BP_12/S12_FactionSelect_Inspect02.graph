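# Level Sequence for a front-end camera move, by its name the second inspect shot on the season 12 faction select screen.
# It possesses the Frontend_EventLevel_Camera actor and keys its transform over 15 frames at 30 fps, easing or moving it to a final position and rotation in about half a second.
# The child CameraComponent holds a constant field of view, and the sequence has no other tracks.
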LevelSequence /MiscReferencersOfCosmetics/Environments/FrontEnd/LevelSequence/BP_12/S12_FactionSelect_Inspect02
bindings [
 CameraComponent {
  id=#1
  parent_id=#2
  possessed_class=CameraComponent
  tracks [{
    type=MovieSceneFloatTrack
    display_name=Field Of View
    sections [{type=MovieSceneFloatSection, channels [None {type=MovieSceneScriptingFloatChannel, default=75, const [75,2]}]}]
   }]
 }
 Camera_Root_Movement {
  id=#2
  possessed_class=Frontend_EventLevel_Camera_C
  tracks [{
    type=MovieScene3DTransformTrack
    sections [{
      type=MovieScene3DTransformSection
      channels [
       Location.X {type=MovieSceneScriptingDoubleChannel, default=1826, keys [[0,1883.6157],[15,1826]]}
       Location.Y {type=MovieSceneScriptingDoubleChannel, default=3762, keys [[0,3684.7812],[15,3762]]}
       Location.Z {type=MovieSceneScriptingDoubleChannel, default=875, keys [[0,915.1245],[15,875]]}
       Rotation.X {type=MovieSceneScriptingDoubleChannel, default=0.7309, keys [[0,0],[15,0.7309]]}
       Rotation.Y {type=MovieSceneScriptingDoubleChannel, default=-1.5914, keys [[0,-10],[15,-1.5914]]}
       Rotation.Z {type=MovieSceneScriptingDoubleChannel, default=-143.6356, keys [[0,-151.2877],[15,-143.6356]]}
       Scale.X {type=MovieSceneScriptingDoubleChannel, default=1}
       Scale.Y {type=MovieSceneScriptingDoubleChannel, default=1}
       Scale.Z {type=MovieSceneScriptingDoubleChannel, default=1}
      ]
     }]
   }]
 }
]
display_rate {denominator=1, numerator=30}
tick_resolution {denominator=1, numerator=24000}
tracks []
