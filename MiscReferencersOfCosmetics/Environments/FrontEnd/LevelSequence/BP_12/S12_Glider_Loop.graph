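# Level sequence for the front-end S12 (BP_12) environment that flies a BP_Glider_Loop actor along a keyframed path.
# A single possessable binding carries one 3D transform track, running about 30 seconds at a 30 fps display rate.
# Location keys sweep the glider back and forth through the scene while rotation keys add banking, pitching and turns, with scale held constant.
# The first and last keyed transforms differ, so the path does not close on itself.
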LevelSequence /MiscReferencersOfCosmetics/Environments/FrontEnd/LevelSequence/BP_12/S12_Glider_Loop
bindings [BP_Glider_Loop {
  id=#1
  possessed_class=BP_Glider_Loop_C
  tracks [{
    type=MovieScene3DTransformTrack
    sections [{
      type=MovieScene3DTransformSection
      channels [
       Location.X {
        type=MovieSceneScriptingDoubleChannel
        default=-604.1815
        thinned_from=11
        keys [[0,-604.1815],[199,-357.2173],[305,-442.9699],[545,-1084.6975],[661.3363,-105.6975],[688,-412.9443],[723,-778.6975],[761,-1436.6975],[801,-882.6975],[905,-882.6975]]
       }
       Location.Y {
        type=MovieSceneScriptingDoubleChannel
        default=-3080.8767
        keys [[0,-4317.877],[199,-1495.0275],[305,-754.5276],[545,-4904.8584],[661.3363,-2345.8584],[688,-2269.2231],[723,-2014.8584],[761,-2398.8584],[801,-2953.8584],[830,-3429.8584],[905,-4740.8584]]
       }
       Location.Z {
        type=MovieSceneScriptingDoubleChannel
        default=996.9873
        keys [[0,1409.9873],[199,616.9188],[305,1118.0479],[545,999.1436],[661.3363,1337.1436],[688,1291.2168],[723,1095.1436],[761,1069.1436],[801,935.1436],[830,935.1436],[905,769.1436]]
       }
       Rotation.X {
        type=MovieSceneScriptingDoubleChannel
        default=2.9093
        thinned_from=15
        keys [[0,0],[98,24.5858],[199,23.8109],[305,2.9093],[545,2.9093],[661.3363,13.1243],[688,13.1243],[723,7.1318],[761,12.6777],[801,3.9645],[830,13.1251],[905,13.1251]]
       }
       Rotation.Y {
        type=MovieSceneScriptingDoubleChannel
        default=0
        keys [[0,0],[98,-19.0359],[199,13.0546],[305,15.7338],[382,30.7123],[415,-14.2296],[456,30.7124],[545,30.7124],[661.3363,-8.0694],[688,-8.0694],[723,16.2167],[761,11.9256],[801,-31.5685],[830,-8.0694],[905,-8.0694]]
       }
       Rotation.Z {
        type=MovieSceneScriptingDoubleChannel
        default=158.1005
        keys [[0,0],[98,-17.6513],[199,-3.4713],[305,179.1478],[382,179.1478],[415,177.6475],[456,179.1478],[545,179.1478],[661.3363,76.2787],[688,76.2787],[723,75.8428],[761,178.706],[801,178.1004],[830,176.2789],[905,176.2789]]
       }
       Scale.X {type=MovieSceneScriptingDoubleChannel, default=1, const [1,15]}
       Scale.Y {type=MovieSceneScriptingDoubleChannel, default=1, const [1,15]}
       Scale.Z {type=MovieSceneScriptingDoubleChannel, default=1, const [1,15]}
      ]
     }]
   }]
 }]
display_rate {denominator=1, numerator=30}
tick_resolution {denominator=1, numerator=24000}
tracks []
