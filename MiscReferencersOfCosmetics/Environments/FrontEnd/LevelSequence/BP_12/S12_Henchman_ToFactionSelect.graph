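# Front-end level sequence for the BP_12 scene that, by its name, moves the camera from the Henchman view to faction select.
# The Frontend_EventLevel_Camera root makes a short keyed move in position, pitch and yaw, with roll held at zero.
# Its child CameraComponent keeps a constant field of view.
# Two event trigger tracks each fire one event, whose targets are not shown.
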LevelSequence /MiscReferencersOfCosmetics/Environments/FrontEnd/LevelSequence/BP_12/S12_Henchman_ToFactionSelect
bindings [
 CameraComponent {
  id=#1
  parent_id=#2
  possessed_class=CameraComponent
  tracks [{
    type=MovieSceneFloatTrack
    display_name=Field Of View
    sections [{type=MovieSceneFloatSection, channels [None {type=MovieSceneScriptingFloatChannel, default=75, const [75,2]}]}]
   }]
 }
 Camera_Root_Movement {
  id=#2
  possessed_class=Frontend_EventLevel_Camera_C
  tracks [{
    type=MovieScene3DTransformTrack
    sections [{
      type=MovieScene3DTransformSection
      channels [
       Location.X {type=MovieSceneScriptingDoubleChannel, default=1883.6157, keys [[0,2043.6157],[15,1883.6157]]}
       Location.Y {type=MovieSceneScriptingDoubleChannel, default=3684.7812, keys [[0,3752.7812],[15,3684.7812]]}
       Location.Z {type=MovieSceneScriptingDoubleChannel, default=915.1245, keys [[0,931.1245],[15,915.1245]]}
       Rotation.X {type=MovieSceneScriptingDoubleChannel, default=0, const [0,2]}
       Rotation.Y {type=MovieSceneScriptingDoubleChannel, default=-10, keys [[0,-2],[15,-10]]}
       Rotation.Z {type=MovieSceneScriptingDoubleChannel, default=-151.2877, keys [[0,-150],[15,-151.2877]]}
       Scale.X {type=MovieSceneScriptingDoubleChannel, default=1}
       Scale.Y {type=MovieSceneScriptingDoubleChannel, default=1}
       Scale.Z {type=MovieSceneScriptingDoubleChannel, default=1}
      ]
     }]
   }]
 }
]
display_rate {denominator=1, numerator=30}
tick_resolution {denominator=1, numerator=24000}
tracks [
 {
  type=MovieSceneEventTrack
  sections [{
    type=MovieSceneEventTriggerSection
    channels [None {type=MovieSceneScriptingEventChannel, const ["<Struct 'MovieSceneEvent' (0x8519a7f5aa0) {}>", 1]}]
   }]
 }
 {
  type=MovieSceneEventTrack
  sections [{
    type=MovieSceneEventTriggerSection
    channels [None {type=MovieSceneScriptingEventChannel, const ["<Struct 'MovieSceneEvent' (0x8519a7f6220) {}>", 1]}]
   }]
 }
]
