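# Front-end Level Sequence for the camera transition from the main room into the BP12 battle pass area.
# The camera pushes toward the elevator door, jumps to a distant location at frame 40, then settles while its FOV narrows.
# Doors open through an animated "Open" material parameter on each side of the jump, with the second door hidden afterward.
# Event tracks trigger two light switchers and sequence-level events, and MPC_Cozy's FadePostBattlepass fades out at the end.
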LevelSequence /MiscReferencersOfCosmetics/Environments/FrontEnd/LevelSequence/BP_12/S12_MainRoom_ToBP
bindings [
 S_BattlePassStar_BP12 {
  id=#1
  parent_id=#2
  possessed_class=StaticMeshComponent
  tracks [{
    type=MovieSceneComponentMaterialTrack
    display_name=Material Element 0
    sections [{type=MovieSceneParameterSection, channels [Open {type=MovieSceneScriptingFloatChannel, default=1, keys [[2,0],[28,1]]}]}]
   }]
 }
 BP_Battlepass_LightSwitcher2 {
  id=#3
  possessed_class=BP_Battlepass_LightSwitcher_C
  tracks [{
    type=MovieSceneEventTrack
    sections [{
      type=MovieSceneEventTriggerSection
      channels [None {type=MovieSceneScriptingEventChannel, const ["<Struct 'MovieSceneEvent' (0x85024359ec0) {}>", 1]}]
     }]
   }]
 }
 CameraComponent {
  id=#4
  parent_id=#5
  possessed_class=CameraComponent
  tracks [{
    type=MovieSceneFloatTrack
    display_name=Field Of View
    sections [{type=MovieSceneFloatSection, channels [None {type=MovieSceneScriptingFloatChannel, default=80, keys [[0,80],[80,75]]}]}]
   }]
 }
 S_Elevator_Door_BP12 {
  id=#6
  parent_id=#2
  possessed_class=StaticMeshComponent
  tracks [
   {
    type=MovieSceneComponentMaterialTrack
    display_name=Material Element 0
    sections [{type=MovieSceneParameterSection, channels [Open {type=MovieSceneScriptingFloatChannel, default=1, keys [[2,0],[28,1]]}]}]
   }
   {
    type=MovieSceneComponentMaterialTrack
    display_name=Material Element 0
    sections [{type=MovieSceneParameterSection, channels [Open {type=MovieSceneScriptingFloatChannel, default=1, keys [[2,0],[28,1]]}]}]
   }
   {
    type=MovieSceneComponentMaterialTrack
    display_name=Material Element 0
    sections [{type=MovieSceneParameterSection, channels [Open {type=MovieSceneScriptingFloatChannel, default=1, keys [[2,0],[28,1]]}]}]
   }
  ]
 }
 BP_Frontend_EventLevel_NavObject_BPDoors {id=#2, possessed_class=BP_Frontend_EventLevel_NavObject_BPDoors_C, tracks []}
 BP_BP12_TransitionDoor2 {id=#7, possessed_class=BP_BP12_TransitionDoor_C, tracks []}
 Camera_Root_Movement {
  id=#5
  possessed_class=Frontend_EventLevel_Camera_C
  tracks [{
    type=MovieScene3DTransformTrack
    sections [{
      type=MovieScene3DTransformSection
      channels [
       Location.X {type=MovieSceneScriptingDoubleChannel, default=53.8848, keys [[0,2821.6787],[24,1806.3861],[40,280],[40.1075,1261.0015],[80,53.8848]]}
       Location.Y {type=MovieSceneScriptingDoubleChannel, default=-73133.4688, keys [[0,-2540.5735],[40,-2559.9995],[40.1075,-73248],[80,-73133.4688]]}
       Location.Z {type=MovieSceneScriptingDoubleChannel, default=160.3502, keys [[0,201.7655],[40,370],[40.1075,248],[80,160.3502]]}
       Rotation.X {type=MovieSceneScriptingDoubleChannel, default=0, thinned_from=4, keys [[0,0],[80,0]]}
       Rotation.Y {type=MovieSceneScriptingDoubleChannel, default=-0.1949, keys [[0,-3.2436],[40,0],[40.1075,0],[80,-0.1949]]}
       Rotation.Z {type=MovieSceneScriptingDoubleChannel, default=-175, thinned_from=4, keys [[0,-179.6847],[40.1075,-179.9998],[80,-175]]}
       Scale.X {type=MovieSceneScriptingDoubleChannel, default=1}
       Scale.Y {type=MovieSceneScriptingDoubleChannel, default=1}
       Scale.Z {type=MovieSceneScriptingDoubleChannel, default=1}
      ]
     }]
   }]
 }
 BP_Battlepass_LightSwitcher_Maproom {
  id=#8
  possessed_class=BP_Battlepass_LightSwitcher_C
  tracks [{
    type=MovieSceneEventTrack
    sections [{
      type=MovieSceneEventTriggerSection
      channels [None {type=MovieSceneScriptingEventChannel, const ["<Struct 'MovieSceneEvent' (0x8502435cb20) {}>", 1]}]
     }]
   }]
 }
 S_BattlePassDoor_Animate_me {
  id=#9
  parent_id=#7
  possessed_class=StaticMeshComponent
  tracks [
   {
    type=MovieSceneComponentMaterialTrack
    display_name=Material Element 0
    sections [{type=MovieSceneParameterSection, channels [Open {type=MovieSceneScriptingFloatChannel, default=1, keys [[40.1075,0],[55,1]]}]}]
   }
   {
    type=MovieSceneComponentMaterialTrack
    display_name=Material Element 0
    sections [{type=MovieSceneParameterSection, channels [Open {type=MovieSceneScriptingFloatChannel, default=1, keys [[40.1075,0],[55,1]]}]}]
   }
   {
    type=MovieSceneComponentMaterialTrack
    display_name=Material Element 0
    sections [{type=MovieSceneParameterSection, channels [Open {type=MovieSceneScriptingFloatChannel, default=1, keys [[40.1075,0],[55,1]]}]}]
   }
   {
    type=MovieSceneVisibilityTrack
    sections [{type=MovieSceneVisibilitySection, channels [None {type=MovieSceneScriptingBoolChannel, default=false, keys [[55,true],[56,false]]}]}]
   }
  ]
 }
 BP_BP12_TransitionDoor {id=#10, possessed_class=BP_BP12_TransitionDoor_C, tracks []}
]
display_rate {denominator=1, numerator=30}
tick_resolution {denominator=1, numerator=24000}
tracks [
 {
  type=MovieSceneEventTrack
  sections [{
    type=MovieSceneEventTriggerSection
    channels [None {type=MovieSceneScriptingEventChannel, const ["<Struct 'MovieSceneEvent' (0x85024359a60) {}>", 1]}]
   }]
 }
 {
  type=MovieSceneEventTrack
  sections [{
    type=MovieSceneEventTriggerSection
    channels [None {type=MovieSceneScriptingEventChannel, const ["<Struct 'MovieSceneEvent' (0x8502435dde0) {}>", 1]}]
   }]
 }
 {
  type=MovieSceneMaterialParameterCollectionTrack
  display_name=MPC_Cozy
  sections [{
    type=MovieSceneParameterSection
    channels [FadePostBattlepass {type=MovieSceneScriptingFloatChannel, default=1, thinned_from=4, keys [[0,1],[64,1],[79.9987,0]]}]
   }]
 }
]
