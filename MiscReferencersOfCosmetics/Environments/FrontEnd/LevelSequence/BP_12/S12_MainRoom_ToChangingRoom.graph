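# Front-end Level Sequence that moves the BP_12 battle pass camera from the main room to the changing room through an elevator-style transition.
# The camera root glides toward the first elevator door, jumps to a second location mid-sequence, then settles as its FOV narrows.
# Material parameter tracks animate the door openings and a downward light pan, masking the jump.
# Event triggers fire on two light switcher actors and on sequence-level tracks.
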LevelSequence /MiscReferencersOfCosmetics/Environments/FrontEnd/LevelSequence/BP_12/S12_MainRoom_ToChangingRoom
bindings [
 S_Elevator_Door_BP12 {
  id=#1
  parent_id=#2
  possessed_class=StaticMeshComponent
  tracks [{
    type=MovieSceneComponentMaterialTrack
    display_name=Material Element 0
    sections [{type=MovieSceneParameterSection, channels [Open {type=MovieSceneScriptingFloatChannel, default=0, keys [[0,0],[23,1],[36.66,1]]}]}]
   }]
 }
 BP_Battlepass_LightSwitcher_Maproom {
  id=#3
  possessed_class=BP_Battlepass_LightSwitcher_C
  tracks [{
    type=MovieSceneEventTrack
    sections [{
      type=MovieSceneEventTriggerSection
      channels [None {type=MovieSceneScriptingEventChannel, const ["<Struct 'MovieSceneEvent' (0x8519a7f8d40) {}>", 1]}]
     }]
   }]
 }
 Camera_Root_Movement {
  id=#4
  possessed_class=Frontend_EventLevel_Camera_C
  tracks [
   {
    type=MovieScene3DTransformTrack
    sections [{
      type=MovieScene3DTransformSection
      channels [
       Location.X {
        type=MovieSceneScriptingDoubleChannel
        default=2264.3303
        keys [[0,2821.6787],[25.4588,1425.1844],[35,800.7001],[35.075,2765.9338],[42,2422.1582],[58,2264.3303]]
       }
       Location.Y {
        type=MovieSceneScriptingDoubleChannel
        default=6327.0493
        thinned_from=7
        keys [[0,-2540.5735],[15,-3000],[35,-3808.2759],[35.075,5972.5884],[43,6214.1201],[58,6327.0493]]
       }
       Location.Z {
        type=MovieSceneScriptingDoubleChannel
        default=899.3929
        keys [[0,201.7655],[7.4425,186.2573],[18,328.9006],[25.67,531.6134],[35,551.1627],[35.075,926],[42.2975,922.0781],[58,899.3929]]
       }
       Rotation.X {type=MovieSceneScriptingDoubleChannel, default=0, thinned_from=7, keys [[0,0],[12,-2.124],[20,-1.4215],[35,0],[58,0]]}
       Rotation.Y {
        type=MovieSceneScriptingDoubleChannel
        default=-3.1963
        thinned_from=7
        keys [[0,-3.2437],[15,-1.3494],[29.9862,-0.0355],[35,0],[42.2975,0],[58,-3.1963]]
       }
       Rotation.Z {
        type=MovieSceneScriptingDoubleChannel
        default=139.195
        keys [[0,-179.6847],[15,-163.8364],[35,-146.1013],[35.075,145.7649],[42,146.0886],[58,139.195]]
       }
       Scale.X {type=MovieSceneScriptingDoubleChannel, default=1, const [1,2]}
       Scale.Y {type=MovieSceneScriptingDoubleChannel, default=1, const [1,2]}
       Scale.Z {type=MovieSceneScriptingDoubleChannel, default=1, const [1,2]}
      ]
     }]
   }
   {type=MovieSceneEventTrack, sections [{type=MovieSceneEventTriggerSection, channels [None {type=MovieSceneScriptingEventChannel}]}]}
  ]
 }
 BP_Elevator_Transition4 {id=#5, possessed_class=BP_Elevator_Transition_C, tracks []}
 Doors {
  id=#6
  parent_id=#5
  possessed_class=StaticMeshComponent
  tracks [
   {
    type=MovieSceneComponentMaterialTrack
    display_name=Material Element 0
    sections [{
      type=MovieSceneParameterSection
      channels [
       Open {type=MovieSceneScriptingFloatChannel, default=1, keys [[37,0],[38,0],[43,1],[43,0]]}
       EnableLightpanDown {type=MovieSceneScriptingFloatChannel, default=0, keys [[37,0],[38,1],[43,0],[43,1]]}
       EnableLightpanUp {type=MovieSceneScriptingFloatChannel, default=0, const [0,3]}
      ]
     }]
   }
   {
    type=MovieSceneComponentMaterialTrack
    display_name=Material Element 0
    sections [{type=MovieSceneParameterSection, channels [Open {type=MovieSceneScriptingFloatChannel, default=1, keys [[37,0],[43,1]]}]}]
   }
  ]
 }
 BP_Frontend_EventLevel_NavObject_Doors2 {id=#2, possessed_class=BP_Frontend_EventLevel_NavObject_C, tracks []}
 CameraComponent {
  id=#7
  parent_id=#4
  possessed_class=CameraComponent
  tracks [{
    type=MovieSceneFloatTrack
    display_name=Field Of View
    sections [{
      type=MovieSceneFloatSection
      channels [None {type=MovieSceneScriptingFloatChannel, default=75, thinned_from=4, keys [[0,80],[35.075,80],[45,75]]}]
     }]
   }]
 }
 BP_Battlepass_LightSwitcher2 {
  id=#8
  possessed_class=BP_Battlepass_LightSwitcher_C
  tracks [{
    type=MovieSceneEventTrack
    sections [{
      type=MovieSceneEventTriggerSection
      channels [None {type=MovieSceneScriptingEventChannel, const ["<Struct 'MovieSceneEvent' (0x8502435dde0) {}>", 1]}]
     }]
   }]
 }
]
display_rate {denominator=1, numerator=30}
tick_resolution {denominator=1, numerator=24000}
tracks [
 {
  type=MovieSceneEventTrack
  sections [{
    type=MovieSceneEventTriggerSection
    channels [None {type=MovieSceneScriptingEventChannel, const ["<Struct 'MovieSceneEvent' (0x8519ac8a0a0) {}>", 1]}]
   }]
 }
 {
  type=MovieSceneEventTrack
  sections [{
    type=MovieSceneEventTriggerSection
    channels [None {type=MovieSceneScriptingEventChannel, const ["<Struct 'MovieSceneEvent' (0x8519ac8fd20) {}>", 1]}]
   }]
 }
]
