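# Front-end level sequence for the camera transition from the main room to the Daily Bounty view, as the asset name indicates.
# A transform track moves and rotates the possessed Frontend_EventLevel_Camera between two keyed poses in a short move, mostly a yaw swing with a slight descent.
# The child CameraComponent's field of view stays constant.
# Two event trigger tracks each fire a single event, whose targets are not visible here.
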LevelSequence /MiscReferencersOfCosmetics/Environments/FrontEnd/LevelSequence/BP_12/S12_MainRoom_ToDailyBounty
bindings [
 Camera_Root_Movement {
  id=#1
  possessed_class=Frontend_EventLevel_Camera_C
  tracks [{
    type=MovieScene3DTransformTrack
    sections [{
      type=MovieScene3DTransformSection
      channels [
       Location.X {type=MovieSceneScriptingDoubleChannel, default=2510.6021, keys [[0,2821.6787],[20,2510.6021]]}
       Location.Y {type=MovieSceneScriptingDoubleChannel, default=-2381.416, keys [[0,-2540.5735],[20,-2381.416]]}
       Location.Z {type=MovieSceneScriptingDoubleChannel, default=146.53, keys [[0,201.7655],[20,146.53]]}
       Rotation.X {type=MovieSceneScriptingDoubleChannel, default=0, keys [[0,0],[20,0]]}
       Rotation.Y {type=MovieSceneScriptingDoubleChannel, default=0, keys [[0,-3.2436],[20,0]]}
       Rotation.Z {type=MovieSceneScriptingDoubleChannel, default=-232, keys [[0,-179.6847],[20,-232]]}
       Scale.X {type=MovieSceneScriptingDoubleChannel, default=1}
       Scale.Y {type=MovieSceneScriptingDoubleChannel, default=1}
       Scale.Z {type=MovieSceneScriptingDoubleChannel, default=1}
      ]
     }]
   }]
 }
 CameraComponent {
  id=#2
  parent_id=#1
  possessed_class=CameraComponent
  tracks [{
    type=MovieSceneFloatTrack
    display_name=Field Of View
    sections [{type=MovieSceneFloatSection, channels [None {type=MovieSceneScriptingFloatChannel, default=80, const [80,2]}]}]
   }]
 }
]
display_rate {denominator=1, numerator=30}
tick_resolution {denominator=1, numerator=24000}
tracks [
 {
  type=MovieSceneEventTrack
  sections [{
    type=MovieSceneEventTriggerSection
    channels [None {type=MovieSceneScriptingEventChannel, const ["<Struct 'MovieSceneEvent' (0x8519ac8c620) {}>", 1]}]
   }]
 }
 {
  type=MovieSceneEventTrack
  sections [{
    type=MovieSceneEventTriggerSection
    channels [None {type=MovieSceneScriptingEventChannel, const ["<Struct 'MovieSceneEvent' (0x8519ac8fd20) {}>", 1]}]
   }]
 }
]
