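# Level sequence for the frontend Battle Pass scene that moves the camera from the main room to the donut room.
# Keyframed transform and field-of-view tracks drive the event-level camera, with a teleport between rooms hidden under a fade to black and back.
# Event triggers on the camera are labeled for stream in and stream out, while others fire on two Battle Pass light switchers and at sequence level.
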LevelSequence /MiscReferencersOfCosmetics/Environments/FrontEnd/LevelSequence/BP_12/S12_MainRoom_ToDonutRoom
bindings [
 BP_Battlepass_LightSwitcher_Maproom {
  id=#1
  possessed_class=BP_Battlepass_LightSwitcher_C
  tracks [{
    type=MovieSceneEventTrack
    sections [{
      type=MovieSceneEventTriggerSection
      channels [None {type=MovieSceneScriptingEventChannel, const ["<Struct 'MovieSceneEvent' (0x8519ac8a0a0) {}>", 1]}]
     }]
   }]
 }
 CameraComponent {
  id=#2
  parent_id=#3
  possessed_class=CameraComponent
  tracks [{
    type=MovieSceneFloatTrack
    display_name=Field Of View
    sections [{type=MovieSceneFloatSection, channels [None {type=MovieSceneScriptingFloatChannel, default=80, keys [[0,80],[76,75]]}]}]
   }]
 }
 Camera_Root_Movement {
  id=#3
  possessed_class=Frontend_EventLevel_Camera_C
  tracks [
   {
    type=MovieScene3DTransformTrack
    sections [{
      type=MovieScene3DTransformSection
      channels [
       Location.X {
        type=MovieSceneScriptingDoubleChannel
        default=2821.6787
        keys [[0,2821.6787],[27,1154.4634],[37,588.705],[38,588.705],[39,275],[40,275],[60,-632.0468],[75,-852.7941]]
       }
       Location.Y {
        type=MovieSceneScriptingDoubleChannel
        default=-2540.5735
        thinned_from=8
        keys [[0,-2540.5735],[27,-3279.0879],[37,-3706.5947],[38,-3706.5947],[39,-8560],[60,-8496.8848],[75,-8562.3223]]
       }
       Location.Z {
        type=MovieSceneScriptingDoubleChannel
        default=201.7655
        keys [[0,201.7655],[27,291.5763],[37,242.4503],[38,242.4503],[39,93.8109],[40,93.8109],[50,136.341],[60,313.3874],[75,365.527]]
       }
       Rotation.X {type=MovieSceneScriptingDoubleChannel, default=0, thinned_from=7, keys [[0,0],[29,-0.3808],[37,-2],[40,-2],[75,0]]}
       Rotation.Y {
        type=MovieSceneScriptingDoubleChannel
        default=-3.2436
        thinned_from=9
        keys [[0,-3.2436],[27,-4.2103],[37,-3.323],[40,-3.323],[46,0],[60,-0.8298],[75,-26.2002]]
       }
       Rotation.Z {
        type=MovieSceneScriptingDoubleChannel
        default=-179.6847
        keys [[0,-179.6847],[24,-154.958],[37,-137.5223],[38,-137.5223],[39,-192.2206],[40,-192.2206],[75,-168.1986]]
       }
       Scale.X {type=MovieSceneScriptingDoubleChannel, default=1}
       Scale.Y {type=MovieSceneScriptingDoubleChannel, default=1}
       Scale.Z {type=MovieSceneScriptingDoubleChannel, default=1}
      ]
     }]
   }
   {
    type=MovieSceneEventTrack
    display_name=Event Stream In Start
    sections [{
      type=MovieSceneEventTriggerSection
      channels [None {type=MovieSceneScriptingEventChannel, const ["<Struct 'MovieSceneEvent' (0x85024359a60) {}>", 1]}]
     }]
   }
   {
    type=MovieSceneEventTrack
    display_name=Event Stream Out
    sections [{
      type=MovieSceneEventTriggerSection
      channels [None {type=MovieSceneScriptingEventChannel, const ["<Struct 'MovieSceneEvent' (0x8519a7f82a0) {}>", 1]}]
     }]
   }
  ]
 }
 BP_Battlepass_LightSwitcher_2 {
  id=#4
  possessed_class=BP_Battlepass_LightSwitcher_C
  tracks [{
    type=MovieSceneEventTrack
    sections [{
      type=MovieSceneEventTriggerSection
      channels [None {type=MovieSceneScriptingEventChannel, const ["<Struct 'MovieSceneEvent' (0x851614d4f60) {}>", 1]}]
     }]
   }]
 }
]
display_rate {denominator=1, numerator=30}
tick_resolution {denominator=1, numerator=24000}
tracks [
 {
  type=MovieSceneFadeTrack
  sections [{type=MovieSceneFadeSection, channels [None {type=MovieSceneScriptingFloatChannel, default=1, keys [[27,0],[35,1],[40,1],[48,0]]}]}]
 }
 {
  type=MovieSceneEventTrack
  sections [{
    type=MovieSceneEventTriggerSection
    channels [None {type=MovieSceneScriptingEventChannel, const ["<Struct 'MovieSceneEvent' (0x84e72598e80) {}>", 1]}]
   }]
 }
 {
  type=MovieSceneEventTrack
  sections [{
    type=MovieSceneEventTriggerSection
    channels [None {type=MovieSceneScriptingEventChannel, const ["<Struct 'MovieSceneEvent' (0x8519ac8c620) {}>", 1]}]
   }]
 }
]
