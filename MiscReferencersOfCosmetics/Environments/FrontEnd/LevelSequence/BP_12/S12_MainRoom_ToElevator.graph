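# Level Sequence for the BP_12 front end that plays a roughly one-second camera move from the main room to the elevator.
# Keyframed transforms on the Frontend_EventLevel_Camera actor drive the travel, rise and turn, while its camera component holds a fixed field of view.
# The elevator door opens during the move through an animated "Open" material parameter on its static mesh, keyed in two material tracks.
# Two event trigger tracks fire events whose targets are not shown.
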LevelSequence /MiscReferencersOfCosmetics/Environments/FrontEnd/LevelSequence/BP_12/S12_MainRoom_ToElevator
bindings [
 BP_Frontend_EventLevel_NavObject_Doors {id=#1, possessed_class=BP_Frontend_EventLevel_NavObject_C, tracks []}
 CameraComponent {
  id=#2
  parent_id=#3
  possessed_class=CameraComponent
  tracks [{
    type=MovieSceneFloatTrack
    display_name=Field Of View
    sections [{type=MovieSceneFloatSection, channels [None {type=MovieSceneScriptingFloatChannel, default=80, const [80,2]}]}]
   }]
 }
 S_Elevator_Door_BP12 {
  id=#4
  parent_id=#1
  possessed_class=StaticMeshComponent
  tracks [
   {
    type=MovieSceneComponentMaterialTrack
    display_name=Material Element 0
    sections [{type=MovieSceneParameterSection, channels [Open {type=MovieSceneScriptingFloatChannel, default=0, keys [[0,0],[16,1]]}]}]
   }
   {
    type=MovieSceneComponentMaterialTrack
    display_name=Material Element 0
    sections [{type=MovieSceneParameterSection, channels [Open {type=MovieSceneScriptingFloatChannel, default=0, keys [[0,0],[16,1]]}]}]
   }
  ]
 }
 Camera_Root_Movement {
  id=#3
  possessed_class=Frontend_EventLevel_Camera_C
  tracks [{
    type=MovieScene3DTransformTrack
    sections [{
      type=MovieScene3DTransformSection
      channels [
       Location.X {type=MovieSceneScriptingDoubleChannel, default=2821.6787, keys [[0,2821.6787],[14,2071.8337],[30,637.1765]]}
       Location.Y {type=MovieSceneScriptingDoubleChannel, default=-2150.447, keys [[0,-2540.5735],[14,-2150.447],[30,-1200.7053]]}
       Location.Z {type=MovieSceneScriptingDoubleChannel, default=201.7655, keys [[0,201.7655],[14,277.4406],[19,460.3505],[30,551.1627]]}
       Rotation.X {type=MovieSceneScriptingDoubleChannel, default=0, keys [[0,0],[8,-1.2235],[30,0]]}
       Rotation.Y {type=MovieSceneScriptingDoubleChannel, default=-3.2436, keys [[0,-3.2436],[14,0.0377],[30,0]]}
       Rotation.Z {type=MovieSceneScriptingDoubleChannel, default=-179.6847, keys [[0,-179.6847],[14,-189.0766],[30,-196.9]]}
       Scale.X {type=MovieSceneScriptingDoubleChannel, default=1}
       Scale.Y {type=MovieSceneScriptingDoubleChannel, default=1}
       Scale.Z {type=MovieSceneScriptingDoubleChannel, default=1}
      ]
     }]
   }]
 }
]
display_rate {denominator=1, numerator=30}
tick_resolution {denominator=1, numerator=24000}
tracks [
 {
  type=MovieSceneEventTrack
  sections [{
    type=MovieSceneEventTriggerSection
    channels [None {type=MovieSceneScriptingEventChannel, const ["<Struct 'MovieSceneEvent' (0x8519a7f82a0) {}>", 1]}]
   }]
 }
 {
  type=MovieSceneEventTrack
  sections [{
    type=MovieSceneEventTriggerSection
    channels [None {type=MovieSceneScriptingEventChannel, const ["<Struct 'MovieSceneEvent' (0x84e72598e80) {}>", 1]}]
   }]
 }
]
