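# Front-end Level Sequence that, by its name, moves the Season 12 lobby camera from the main room to the event view.
# A transform track moves and turns the Frontend_EventLevel_Camera actor over a short 20-frame move, adding a slight roll that peaks midway and settles level.
# Its child CameraComponent narrows field of view slightly over the same frames.
# Two event trigger tracks also fire, though their bound functions are not shown.
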LevelSequence /MiscReferencersOfCosmetics/Environments/FrontEnd/LevelSequence/BP_12/S12_MainRoom_ToEvent
bindings [
 CameraComponent {
  id=#1
  parent_id=#2
  possessed_class=CameraComponent
  tracks [{
    type=MovieSceneFloatTrack
    display_name=Field Of View
    sections [{type=MovieSceneFloatSection, channels [None {type=MovieSceneScriptingFloatChannel, default=75, keys [[0,80],[20,75]]}]}]
   }]
 }
 Camera_Root_Movement {
  id=#2
  possessed_class=Frontend_EventLevel_Camera_C
  tracks [{
    type=MovieScene3DTransformTrack
    sections [{
      type=MovieScene3DTransformSection
      channels [
       Location.X {type=MovieSceneScriptingDoubleChannel, default=2448.407, keys [[0,2821.6787],[20,2448.407]]}
       Location.Y {type=MovieSceneScriptingDoubleChannel, default=-2786.1743, keys [[0,-2540.5735],[20,-2786.1743]]}
       Location.Z {type=MovieSceneScriptingDoubleChannel, default=129.9053, keys [[0,201.7655],[20,129.9053]]}
       Rotation.X {type=MovieSceneScriptingDoubleChannel, default=0, keys [[0,0],[10,1.5],[20,0]]}
       Rotation.Y {type=MovieSceneScriptingDoubleChannel, default=-3.2436, keys [[0,-3.2436],[20,0]]}
       Rotation.Z {type=MovieSceneScriptingDoubleChannel, default=-154.9843, keys [[0,-179.6847],[20,-154.9843]]}
       Scale.X {type=MovieSceneScriptingDoubleChannel, default=1}
       Scale.Y {type=MovieSceneScriptingDoubleChannel, default=1}
       Scale.Z {type=MovieSceneScriptingDoubleChannel, default=1}
      ]
     }]
   }]
 }
]
display_rate {denominator=1, numerator=30}
tick_resolution {denominator=1, numerator=24000}
tracks [
 {
  type=MovieSceneEventTrack
  sections [{
    type=MovieSceneEventTriggerSection
    channels [None {type=MovieSceneScriptingEventChannel, const ["<Struct 'MovieSceneEvent' (0x8519ade0640) {}>", 1]}]
   }]
 }
 {
  type=MovieSceneEventTrack
  sections [{
    type=MovieSceneEventTriggerSection
    channels [None {type=MovieSceneScriptingEventChannel, const ["<Struct 'MovieSceneEvent' (0x8519ade01e0) {}>", 1]}]
   }]
 }
]
