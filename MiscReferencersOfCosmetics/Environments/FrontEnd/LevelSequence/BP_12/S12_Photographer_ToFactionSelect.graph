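# Level Sequence for the BP_12 front end that moves the event camera, apparently from the Photographer view to faction select.
# A short transform track eases Frontend_EventLevel_Camera_C through a small position shift and a yaw swing while leveling pitch.
# Its child CameraComponent holds a constant field of view.
# Two master event tracks each fire a single trigger, whose target functions are not shown.
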LevelSequence /MiscReferencersOfCosmetics/Environments/FrontEnd/LevelSequence/BP_12/S12_Photographer_ToFactionSelect
bindings [
 CameraComponent {
  id=#1
  parent_id=#2
  possessed_class=CameraComponent
  tracks [{
    type=MovieSceneFloatTrack
    display_name=Field Of View
    sections [{type=MovieSceneFloatSection, channels [None {type=MovieSceneScriptingFloatChannel, default=80, const [85,2]}]}]
   }]
 }
 Camera_Root_Movement {
  id=#2
  possessed_class=Frontend_EventLevel_Camera_C
  tracks [{
    type=MovieScene3DTransformTrack
    sections [{
      type=MovieScene3DTransformSection
      channels [
       Location.X {type=MovieSceneScriptingDoubleChannel, default=2527.9539, keys [[0,2527.9539],[15,2502.5781]]}
       Location.Y {type=MovieSceneScriptingDoubleChannel, default=-106702.7891, keys [[0,-106702.7891],[15,-106799.1641]]}
       Location.Z {type=MovieSceneScriptingDoubleChannel, default=104.2625, keys [[0,104.2625],[15,107.8874]]}
       Rotation.X {type=MovieSceneScriptingDoubleChannel, default=0, keys [[0,0],[15,0]]}
       Rotation.Y {type=MovieSceneScriptingDoubleChannel, default=1.7322, keys [[0,1.7322],[15,0]]}
       Rotation.Z {type=MovieSceneScriptingDoubleChannel, default=-149.2204, keys [[0,-149.2204],[15,-175.5904]]}
       Scale.X {type=MovieSceneScriptingDoubleChannel, default=1}
       Scale.Y {type=MovieSceneScriptingDoubleChannel, default=1}
       Scale.Z {type=MovieSceneScriptingDoubleChannel, default=1}
      ]
     }]
   }]
 }
]
display_rate {denominator=1, numerator=30}
tick_resolution {denominator=1, numerator=24000}
tracks [
 {
  type=MovieSceneEventTrack
  sections [{
    type=MovieSceneEventTriggerSection
    channels [None {type=MovieSceneScriptingEventChannel, const ["<Struct 'MovieSceneEvent' (0x8519ade1720) {}>", 1]}]
   }]
 }
 {
  type=MovieSceneEventTrack
  sections [{
    type=MovieSceneEventTriggerSection
    channels [None {type=MovieSceneScriptingEventChannel, const ["<Struct 'MovieSceneEvent' (0x8519ade10e0) {}>", 1]}]
   }]
 }
]
