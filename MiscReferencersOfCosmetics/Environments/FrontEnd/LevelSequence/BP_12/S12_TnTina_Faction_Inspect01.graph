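# Front-end camera sequence for a Season 12 TnTina faction inspect view, judging by its name and BP_12 folder.
# It possesses a Frontend_EventLevel_Camera actor and makes a short keyed move from frame 0 to 15, shifting location, pitch and yaw into the framing.
# The child CameraComponent holds a constant field of view.
# Two root event trigger tracks each carry one event, with no visible target function.
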LevelSequence /MiscReferencersOfCosmetics/Environments/FrontEnd/LevelSequence/BP_12/S12_TnTina_Faction_Inspect01
bindings [
 CameraComponent {
  id=#1
  parent_id=#2
  possessed_class=CameraComponent
  tracks [{
    type=MovieSceneFloatTrack
    display_name=Field Of View
    sections [{type=MovieSceneFloatSection, channels [None {type=MovieSceneScriptingFloatChannel, default=80, const [80,2]}]}]
   }]
 }
 Camera_Root_Movement {
  id=#2
  possessed_class=Frontend_EventLevel_Camera_C
  tracks [{
    type=MovieScene3DTransformTrack
    sections [{
      type=MovieScene3DTransformSection
      channels [
       Location.X {type=MovieSceneScriptingDoubleChannel, default=2224.4426, keys [[0,2299.2668],[15,2224.4426]]}
       Location.Y {type=MovieSceneScriptingDoubleChannel, default=10644.9531, keys [[0,10666.5801],[15,10644.9531]]}
       Location.Z {type=MovieSceneScriptingDoubleChannel, default=870.8293, keys [[0,864.2753],[15,870.8293]]}
       Rotation.X {type=MovieSceneScriptingDoubleChannel, default=0, keys [[0,0],[15,0]]}
       Rotation.Y {type=MovieSceneScriptingDoubleChannel, default=-2.3997, keys [[0,-0.8],[15,-2.3997]]}
       Rotation.Z {type=MovieSceneScriptingDoubleChannel, default=-237.4128, keys [[0,-231.4105],[15,-237.4128]]}
       Scale.X {type=MovieSceneScriptingDoubleChannel, default=1}
       Scale.Y {type=MovieSceneScriptingDoubleChannel, default=1}
       Scale.Z {type=MovieSceneScriptingDoubleChannel, default=1}
      ]
     }]
   }]
 }
]
display_rate {denominator=1, numerator=30}
tick_resolution {denominator=1, numerator=24000}
tracks [
 {
  type=MovieSceneEventTrack
  sections [{
    type=MovieSceneEventTriggerSection
    channels [None {type=MovieSceneScriptingEventChannel, const ["<Struct 'MovieSceneEvent' (0x8519ade10e0) {}>", 1]}]
   }]
 }
 {
  type=MovieSceneEventTrack
  sections [{
    type=MovieSceneEventTriggerSection
    channels [None {type=MovieSceneScriptingEventChannel, const ["<Struct 'MovieSceneEvent' (0x8519ade1cc0) {}>", 1]}]
   }]
 }
]
