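# Front-end camera Level Sequence for the second TnTina faction inspect view in the Season 12 (BP_12) set.
# A transform track eases or moves the possessed Frontend_EventLevel_Camera_C from a start to an end pose over 15 frames at 30 fps.
# Its child CameraComponent has a Field of View track holding a constant value.
# Two root event trigger tracks fire events whose handlers are not shown here.
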LevelSequence /MiscReferencersOfCosmetics/Environments/FrontEnd/LevelSequence/BP_12/S12_TnTina_Faction_Inspect02
bindings [
 CameraComponent {
  id=#1
  parent_id=#2
  possessed_class=CameraComponent
  tracks [{
    type=MovieSceneFloatTrack
    display_name=Field Of View
    sections [{type=MovieSceneFloatSection, channels [None {type=MovieSceneScriptingFloatChannel, default=80, const [80,2]}]}]
   }]
 }
 Camera_Root_Movement {
  id=#2
  possessed_class=Frontend_EventLevel_Camera_C
  tracks [{
    type=MovieScene3DTransformTrack
    sections [{
      type=MovieScene3DTransformSection
      channels [
       Location.X {type=MovieSceneScriptingDoubleChannel, default=2362.0876, keys [[0,2299.2668],[15,2362.0876]]}
       Location.Y {type=MovieSceneScriptingDoubleChannel, default=10740.2568, keys [[0,10666.5801],[15,10740.2568]]}
       Location.Z {type=MovieSceneScriptingDoubleChannel, default=875.236, keys [[0,864.2753],[15,875.236]]}
       Rotation.X {type=MovieSceneScriptingDoubleChannel, default=-1.8265, keys [[0,0],[15,-1.8265]]}
       Rotation.Y {type=MovieSceneScriptingDoubleChannel, default=-3.2753, keys [[0,-0.8],[15,-3.2753]]}
       Rotation.Z {type=MovieSceneScriptingDoubleChannel, default=-218.9743, keys [[0,-231.4105],[15,-218.9743]]}
       Scale.X {type=MovieSceneScriptingDoubleChannel, default=1}
       Scale.Y {type=MovieSceneScriptingDoubleChannel, default=1}
       Scale.Z {type=MovieSceneScriptingDoubleChannel, default=1}
      ]
     }]
   }]
 }
]
display_rate {denominator=1, numerator=30}
tick_resolution {denominator=1, numerator=24000}
tracks [
 {
  type=MovieSceneEventTrack
  sections [{
    type=MovieSceneEventTriggerSection
    channels [None {type=MovieSceneScriptingEventChannel, const ["<Struct 'MovieSceneEvent' (0x8519ade1ae0) {}>", 1]}]
   }]
 }
 {
  type=MovieSceneEventTrack
  sections [{
    type=MovieSceneEventTriggerSection
    channels [None {type=MovieSceneScriptingEventChannel, const ["<Struct 'MovieSceneEvent' (0x8519ade1c20) {}>", 1]}]
   }]
 }
]
