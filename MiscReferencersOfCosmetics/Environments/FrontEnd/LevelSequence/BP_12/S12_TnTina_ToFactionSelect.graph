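# Front-end level sequence for BP_12 that, by its name, moves the camera from the TnTina view to faction select.
# A short keyed transform track shifts and slightly turns the Frontend_EventLevel_Camera actor, while its child camera component holds a constant field of view.
# Two sequence-level event trigger tracks each carry an event, though their target functions are not identified here.
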LevelSequence /MiscReferencersOfCosmetics/Environments/FrontEnd/LevelSequence/BP_12/S12_TnTina_ToFactionSelect
bindings [
 CameraComponent {
  id=#1
  parent_id=#2
  possessed_class=CameraComponent
  tracks [{
    type=MovieSceneFloatTrack
    display_name=Field Of View
    sections [{type=MovieSceneFloatSection, channels [None {type=MovieSceneScriptingFloatChannel, default=80, const [80,2]}]}]
   }]
 }
 Camera_Root_Movement {
  id=#2
  possessed_class=Frontend_EventLevel_Camera_C
  tracks [{
    type=MovieScene3DTransformTrack
    sections [{
      type=MovieScene3DTransformSection
      channels [
       Location.X {type=MovieSceneScriptingDoubleChannel, default=2299.2668, keys [[0,2378.3931],[15,2299.2668]]}
       Location.Y {type=MovieSceneScriptingDoubleChannel, default=10666.5801, keys [[0,10649.8271],[15,10666.5801]]}
       Location.Z {type=MovieSceneScriptingDoubleChannel, default=864.2753, keys [[0,876.3943],[15,864.2753]]}
       Rotation.X {type=MovieSceneScriptingDoubleChannel, default=0, keys [[0,0],[15,0]]}
       Rotation.Y {type=MovieSceneScriptingDoubleChannel, default=-0.8, keys [[0,0.5995],[15,-0.8]]}
       Rotation.Z {type=MovieSceneScriptingDoubleChannel, default=-231.4105, keys [[0,-228.408],[15,-231.4105]]}
       Scale.X {type=MovieSceneScriptingDoubleChannel, default=1}
       Scale.Y {type=MovieSceneScriptingDoubleChannel, default=1}
       Scale.Z {type=MovieSceneScriptingDoubleChannel, default=1}
      ]
     }]
   }]
 }
]
display_rate {denominator=1, numerator=30}
tick_resolution {denominator=1, numerator=24000}
tracks [
 {
  type=MovieSceneEventTrack
  sections [{
    type=MovieSceneEventTriggerSection
    channels [None {type=MovieSceneScriptingEventChannel, const ["<Struct 'MovieSceneEvent' (0x8519ade1220) {}>", 1]}]
   }]
 }
 {
  type=MovieSceneEventTrack
  sections [{
    type=MovieSceneEventTriggerSection
    channels [None {type=MovieSceneScriptingEventChannel, const ["<Struct 'MovieSceneEvent' (0x8519ade10e0) {}>", 1]}]
   }]
 }
]
